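# Niagara one-shot burst effect for the Sitcom emote: quick glow flashes, a growing mesh and expanding rotating release sprites, mostly shifting warm to blue.
# Only Condensed_BurstRelease is enabled, consolidating the disabled Glowy, Calm_Release and Additional_Glow emitters.
# Seven instantaneous spawn groups write VisibilityTag and pick lifetime, size, camera offset and rotation through MergeHelper switches.
# Separate per-layer color attributes serve its six sprite renderers and one mesh renderer.
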NiagaraSystem /BRCosmetics/Animation/Game/MainPlayer/Emotes/Sitcom/FX/NS_Emote_Sitcom_Burst
dependencies {
 usedDataInterfaces [&NiagaraDataInterfaceCurve, &NiagaraDataInterfaceColorCurve]
 usedDynamicInputs []
 usedModules [&SystemState]
}
emitters {
 Additional_Glow {
  bEnabled=false
  emitterUpdateScript [
   &EmitterState()
   &SpawnBurst_Instantaneous(Use Spawn Probability=false, UseLoopCountLimit=false)
   &SpawnBurst_Instantaneous(Spawn Group=1, Use Spawn Probability=false, UseLoopCountLimit=false) {module=SpawnBurst_Instantaneous001}
  ]
  particleSpawnScript [
   /SetVariables(Particles.VisibilityTag=&SpawnGroup()) {isSetParameters=true}
   &InitializeParticle(
    Color Mode=::NewEnumerator0
    Lifetime=0.6
    Sprite Size Mode=::NewEnumerator3
    Uniform Sprite Size=&MergeHelper_FloatSwitch(0=90, 1=165, 2=?, 3=?, Index=&SpawnGroup(), Num=::NewEnumerator0)
    UsePositionOffset=false
    Write Material Random=true
   )
   &CameraOffset(Camera Offset Amount=5)
  ]
  particleUpdateScript [
   &ParticleState()
   &ScaleSpriteSize(Uniform Curve Sprite Scale=DataInterface() {propertyValues {Curve {
       defaultValue=0
       keys [
        {arriveTangent=59.7205, interpMode=RCIM_Cubic, leaveTangent=59.7205, tangentMode=RCTM_User, time=0, value=0}
        {arriveTangent=-8.6051, interpMode=RCIM_Cubic, leaveTangent=-8.6051, tangentMode=RCTM_User, time=0.0821, value=0.9847}
        {arriveTangent=-1.7363, interpMode=RCIM_Cubic, leaveTangent=-1.7363, tangentMode=RCTM_User, time=0.1906, value=0.3836}
        {arriveTangent=-0.0405, interpMode=RCIM_Cubic, leaveTangent=-0.0405, tangentMode=RCTM_User, time=1, value=0}
       ]
      }}})
   &ScaleColor(
    Linear Color Curve=DataInterface() {propertyValues {
      AlphaCurve {keys [<at 0 0.3>,<at 0.239 0.05>,<at 0.6226 0.01>,<at 1 0>]}
      BlueCurve {keys [<at 0.0739 0>,<at 0.1334 0>,<at 0.1916 0.9175>,<at 0.2572 1>,<at 0.8277 1>]}
      GreenCurve {keys [<at 0.0739 0.0072>,<at 0.1334 0.3226>,<at 0.1916 0.2172>,<at 0.2572 0.4957>,<at 0.8277 0.8782>]}
      RedCurve {keys [<at 0.0739 1>,<at 0.1334 0.9167>,<at 0.1916 0.9479>,<at 0.2572 0>,<at 0.8277 0.525>]}
     }}
    Scale Mode=::NewEnumerator2
   )
   /SetVariables(Particles.Linear Color=&ColorFromCurve(ColorCurve=DataInterface() {propertyValues {
       AlphaCurve {keys [<at 0 1 RCIM_Cubic>,<at 0.9528 0 RCIM_Cubic>]}
       BlueCurve {keys [<at 0.0558 0 RCIM_Cubic>,<at 0.5659 0.0865 RCIM_Cubic>]}
       GreenCurve {keys [<at 0.0558 0.0041 RCIM_Cubic>,<at 0.5659 0.0286 RCIM_Cubic>]}
       RedCurve {keys [<at 0.0558 0.1198 RCIM_Cubic>,<at 0.5659 0 RCIM_Cubic>]}
      }})) {isSetParameters=true}
  ]
  rendererClasses [&NiagaraSpriteRendererProperties]
  renderers [
   {rendererClass=&NiagaraSpriteRendererProperties}
   {rendererClass=&NiagaraSpriteRendererProperties, rendererIndex=1}
  ]
 }
 Calm_Release {
  bEnabled=false
  emitterUpdateScript [
   &EmitterState()
   &SpawnBurst_Instantaneous(Use Spawn Probability=false, UseLoopCountLimit=false)
   &SpawnBurst_Instantaneous(Spawn Group=1, Use Spawn Probability=false, UseLoopCountLimit=false) {module=SpawnBurst_Instantaneous001}
  ]
  particleSpawnScript [
   /SetVariables(Particles.VisibilityTag=&SpawnGroup()) {isSetParameters=true}
   &InitializeParticle(
    Color=&MergeHelper_ColorSwitch(
     0=<rgba 1.5771 1.8302 2 0.8191>
     1=<rgba 0.4523 0.7802 1 0.2>
     2=?
     3=?
     Index=&SpawnGroup()
     Num=::NewEnumerator0
    )
    Lifetime=&MergeHelper_FloatSwitch(0=0.75, 1=1, 2=?, 3=?, Index=&SpawnGroup(), Num=::NewEnumerator0)
    Sprite Rotation Mode=::NewEnumerator2
    Sprite Size Mode=::NewEnumerator3
    Uniform Sprite Size=30
    UsePositionOffset=false
    Write Material Random=true
   )
   &CameraOffset(Camera Offset Amount=5)
  ]
  particleUpdateScript [
   &ParticleState()
   &SpriteRotationRate(Rotation Rate=&RandomRangeFloat(Minimum=-10, Maximum=10))
   &ScaleSpriteSize(
    Uniform Curve Scale=3.8
    Uniform Curve Sprite Scale=DataInterface() {propertyValues {Curve {
       defaultValue=0
       keys [
        {arriveTangent=75.0186, interpMode=RCIM_Cubic, leaveTangent=75.0186, tangentMode=RCTM_User, time=0, value=0}
        {arriveTangent=-36.4448, interpMode=RCIM_Cubic, leaveTangent=0.6918, tangentMode=RCTM_Break, time=0.0163, value=0.2538}
        {arriveTangent=1.3706, interpMode=RCIM_Cubic, leaveTangent=1.3706, time=0.6006, value=1.1605}
        {arriveTangent=0.0224, interpMode=RCIM_Cubic, leaveTangent=0.0224, tangentMode=RCTM_User, time=0.8989, value=1.4634}
       ]
      }}}
   )
   &ScaleColor(
    Scale Alpha=&FloatFromCurve(FloatCurve=DataInterface() {propertyValues {Curve {keys [
         {arriveTangent=63.1994, interpMode=RCIM_Cubic, leaveTangent=63.1994, tangentMode=RCTM_User, time=-0.0023, value=0}
         {interpMode=RCIM_Cubic, leaveTangent=-1.275, tangentMode=RCTM_Break, time=0.0567, value=1}
         {arriveTangent=-1.0601, interpMode=RCIM_Cubic, leaveTangent=-1.0601, time=0.3476, value=0.2706}
         <at 1 0>
        ]}}})
    Scale RGB=&MakeVectorFromLinearColorRGB(Color=&ColorFromCurve(ColorCurve=DataInterface() {propertyValues {
        AlphaCurve {keys []}
        BlueCurve {keys [<at 0.0292 0 RCIM_Cubic>,<at 0.5331 2 RCIM_Cubic>]}
        GreenCurve {keys [<at 0.0292 0 RCIM_Cubic>,<at 0.5331 1.8345 RCIM_Cubic>]}
        RedCurve {keys [<at 0.0292 4 RCIM_Cubic>,<at 0.5331 1.1934 RCIM_Cubic>]}
       }}))
    ScaleA=true
    ScaleRGB=true
   )
   /SetVariables(
    Particles.ScaleColor2:LinearColor=&MakeLinearColorFromVectorAndFloat(
     "Vector (RGB)"=&MakeVectorFromLinearColorRGB(Color=<rgba 0.3639 0.8084 0.8323 0.03>)
     "Float (Alpha)"=&FloatFromCurve(FloatCurve=DataInterface() {propertyValues {Curve {
         defaultValue=1
         keys [{interpMode=RCIM_Cubic, tangentMode=RCTM_User, time=0, value=1}, {interpMode=RCIM_Cubic, tangentMode=RCTM_User, time=1, value=0}]
        }}})
    )
    Particles.SpriteSize2=&Multiply_Vector2D_ByFloat(
     Float=&FloatFromCurve(
      FloatCurve=DataInterface() {propertyValues {Curve {
         defaultValue=1
         keys [
          {arriveTangent=2.2218, interpMode=RCIM_Cubic, leaveTangent=2.2218, tangentMode=RCTM_User, time=0, value=0}
          {arriveTangent=0.4821, interpMode=RCIM_Cubic, leaveTangent=0.4821, tangentMode=RCTM_User, time=0.3499, value=0.6425}
          {arriveTangent=0.3373, interpMode=RCIM_Cubic, leaveTangent=0.3373, tangentMode=RCTM_User, time=1.4958, value=1.1601}
         ]
        }}}
      Scale Curve=1.9
     )
     Vector2D=$Particles.Initial.SpriteSize
    )
   ) {isSetParameters=true}
  ]
  rendererClasses [&NiagaraSpriteRendererProperties]
  renderers [
   {rendererClass=&NiagaraSpriteRendererProperties}
   {rendererClass=&NiagaraSpriteRendererProperties, rendererIndex=1}
  ]
 }
 Condensed_BurstRelease {
  emitterUpdateScript [
   &EmitterState()
   &SpawnBurst_Instantaneous(Use Spawn Probability=false, UseLoopCountLimit=false) {module=SpawnBurst_Instantaneous002}
   &SpawnBurst_Instantaneous(Spawn Group=1, Use Spawn Probability=false, UseLoopCountLimit=false) {module=SpawnBurst_Instantaneous001}
   &SpawnBurst_Instantaneous(Spawn Group=2, Use Spawn Probability=false, UseLoopCountLimit=false)
   &SpawnBurst_Instantaneous(Spawn Group=3, Use Spawn Probability=false, UseLoopCountLimit=false) {module=SpawnBurst_Instantaneous003}
   &SpawnBurst_Instantaneous(Spawn Group=4, Use Spawn Probability=false, UseLoopCountLimit=false) {module=SpawnBurst_Instantaneous004}
   &SpawnBurst_Instantaneous(Spawn Group=5, Use Spawn Probability=false, UseLoopCountLimit=false) {module=SpawnBurst_Instantaneous005}
   &SpawnBurst_Instantaneous(Spawn Group=6, Use Spawn Probability=false, UseLoopCountLimit=false) {module=SpawnBurst_Instantaneous006}
  ]
  particleSpawnScript [
   /SetVariables(Particles.VisibilityTag=&SpawnGroup()) {isSetParameters=true}
   &InitializeParticle(
    Color Mode=::NewEnumerator0
    Lifetime=&MergeHelper_FloatSwitch(0=0.8, 1=0.5, 2=0.8, 3=0.6, 4=0.6, 5=0.75, 6=0.75, Index=&SpawnGroup(), Num=::NewEnumerator5)
    Mesh Scale=<0.28 0.4 0.4>
    Mesh Scale Mode=::NewEnumerator4
    Sprite Rotation Mode=::NewEnumerator2
    Sprite Size Mode=::NewEnumerator3
    Uniform Sprite Size=&MergeHelper_FloatSwitch(0=60, 1=0, 2=50, 3=90, 4=100, 5=30, 6=30, Index=&SpawnGroup(), Num=::NewEnumerator5)
    UsePositionOffset=false
    Write Material Random=true
   )
   &CameraOffset(Camera Offset Amount=&MergeHelper_FloatSwitch(0=-3, 1=-3, 2=-3, 3=5, 4=5, 5=5, 6=5, Index=&SpawnGroup(), Num=::NewEnumerator5))
  ]
  particleUpdateScript [
   &ParticleState()
   &ScaleSpriteSize(
    Scale Sprite Size Mode=::NewEnumerator0
    Uniform Scale Factor=&MergeHelper_FloatSwitch(
     0=&FloatFromCurve(FloatCurve=DataInterface() {propertyValues {Curve {
         defaultValue=0
         keys [
          {arriveTangent=52.4903, interpMode=RCIM_Cubic, leaveTangent=52.4903, tangentMode=RCTM_User, time=0, value=0}
          {arriveTangent=-24.6613, interpMode=RCIM_Cubic, leaveTangent=-5.5991, tangentMode=RCTM_Break, time=0.0549, value=1.1994}
          {arriveTangent=-1.2691, interpMode=RCIM_Cubic, leaveTangent=-1.2691, time=0.2916, value=0.2084}
          {tangentMode=RCTM_User, time=1, value=0}
         ]
        }}})
     1=0
     2=&FloatFromCurve(FloatCurve=DataInterface() {propertyValues {Curve {
         defaultValue=0
         keys [
          {arriveTangent=52.4903, interpMode=RCIM_Cubic, leaveTangent=52.4903, tangentMode=RCTM_User, time=0, value=0}
          {arriveTangent=-24.6613, interpMode=RCIM_Cubic, leaveTangent=-5.5991, tangentMode=RCTM_Break, time=0.0549, value=1.1994}
          {arriveTangent=-1.2691, interpMode=RCIM_Cubic, leaveTangent=-1.2691, time=0.2916, value=0.2084}
          {tangentMode=RCTM_User, time=1, value=0}
         ]
        }}})
     3=&FloatFromCurve(FloatCurve=DataInterface() {propertyValues {Curve {
         defaultValue=0
         keys [
          {arriveTangent=59.7205, interpMode=RCIM_Cubic, leaveTangent=59.7205, tangentMode=RCTM_User, time=0, value=0}
          {arriveTangent=-8.6051, interpMode=RCIM_Cubic, leaveTangent=-8.6051, tangentMode=RCTM_User, time=0.0821, value=0.9847}
          {arriveTangent=-1.7363, interpMode=RCIM_Cubic, leaveTangent=-1.7363, tangentMode=RCTM_User, time=0.1906, value=0.3836}
          {arriveTangent=-0.0405, interpMode=RCIM_Cubic, leaveTangent=-0.0405, tangentMode=RCTM_User, time=1, value=0}
         ]
        }}})
     4=&FloatFromCurve(FloatCurve=DataInterface() {propertyValues {Curve {
         defaultValue=0
         keys [
          {arriveTangent=59.7205, interpMode=RCIM_Cubic, leaveTangent=59.7205, tangentMode=RCTM_User, time=0, value=0}
          {arriveTangent=-8.6051, interpMode=RCIM_Cubic, leaveTangent=-8.6051, tangentMode=RCTM_User, time=0.0821, value=0.9847}
          {arriveTangent=-1.7363, interpMode=RCIM_Cubic, leaveTangent=-1.7363, tangentMode=RCTM_User, time=0.1906, value=0.3836}
          {arriveTangent=-0.0405, interpMode=RCIM_Cubic, leaveTangent=-0.0405, tangentMode=RCTM_User, time=1, value=0}
         ]
        }}})
     5=&FloatFromCurve(
      FloatCurve=DataInterface() {propertyValues {Curve {
         defaultValue=0
         keys [
          {arriveTangent=75.0186, interpMode=RCIM_Cubic, leaveTangent=75.0186, tangentMode=RCTM_User, time=0, value=0}
          {arriveTangent=-36.4448, interpMode=RCIM_Cubic, leaveTangent=0.6918, tangentMode=RCTM_Break, time=0.0163, value=0.2538}
          {arriveTangent=1.3706, interpMode=RCIM_Cubic, leaveTangent=1.3706, time=0.6006, value=1.1605}
          {arriveTangent=0.0224, interpMode=RCIM_Cubic, leaveTangent=0.0224, tangentMode=RCTM_User, time=0.8989, value=1.4634}
         ]
        }}}
      Scale Curve=3.8
     )
     6=&FloatFromCurve(
      FloatCurve=DataInterface() {propertyValues {Curve {
         defaultValue=1
         keys [
          {arriveTangent=2.2218, interpMode=RCIM_Cubic, leaveTangent=2.2218, tangentMode=RCTM_User, time=0, value=0}
          {arriveTangent=0.4821, interpMode=RCIM_Cubic, leaveTangent=0.4821, tangentMode=RCTM_User, time=0.3499, value=0.6425}
          {arriveTangent=0.3373, interpMode=RCIM_Cubic, leaveTangent=0.3373, tangentMode=RCTM_User, time=1.4958, value=1.1601}
         ]
        }}}
      Scale Curve=1.9
     )
     Index=&SpawnGroup()
     Num=::NewEnumerator5
    )
   )
   /SetVariables(
    Particles.BG_Glow_Bright Color:LinearColor=&ColorFromCurve(ColorCurve=DataInterface() {propertyValues {
       AlphaCurve {keys [<at 0 0.3>,<at 0.239 0.05>,<at 0.6226 0.01>,<at 1 0>]}
       BlueCurve {keys [<at 0.0739 0>,<at 0.1334 0>,<at 0.1916 0.9175>,<at 0.2572 1>,<at 0.8277 1>]}
       GreenCurve {keys [<at 0.0739 0.0072>,<at 0.1334 0.3226>,<at 0.1916 0.2172>,<at 0.2572 0.4957>,<at 0.8277 0.8782>]}
       RedCurve {keys [<at 0.0739 1>,<at 0.1334 0.9167>,<at 0.1916 0.9479>,<at 0.2572 0>,<at 0.8277 0.525>]}
      }})
    Particles.BG_Glow_Dark Color:LinearColor=&ColorFromCurve(ColorCurve=DataInterface() {propertyValues {
       AlphaCurve {keys [<at 0 1 RCIM_Cubic>,<at 0.9528 0 RCIM_Cubic>]}
       BlueCurve {keys [<at 0.0558 0 RCIM_Cubic>,<at 0.5659 0.0865 RCIM_Cubic>]}
       GreenCurve {keys [<at 0.0558 0.0041 RCIM_Cubic>,<at 0.5659 0.0286 RCIM_Cubic>]}
       RedCurve {keys [<at 0.0558 0.1198 RCIM_Cubic>,<at 0.5659 0 RCIM_Cubic>]}
      }})
    Particles.Glow_Mesh Color:LinearColor=&ColorFromCurve(ColorCurve=DataInterface() {propertyValues {
       AlphaCurve {keys [<at 0 0.7 RCIM_Cubic>,<at 0.1384 0.1>,<at 0.9724 0 RCIM_Cubic>]}
       BlueCurve {keys [
         <at 0 0 RCIM_Cubic>
         <at 0.2011 0.6667>
         {arriveTangent=0.8019, interpMode=RCIM_Cubic, leaveTangent=0.8019, time=0.4482, value=0.6146}
         <at 0.7469 0.599>
        ]}
       GreenCurve {keys [
         <at 0 0.0191 RCIM_Cubic>
         <at 0.2011 0.1998>
         {arriveTangent=0.4637, interpMode=RCIM_Cubic, leaveTangent=0.4637, time=0.4482, value=0.2117}
         <at 0.7469 0.3654>
        ]}
       RedCurve {keys [
         <at 0 0.5885 RCIM_Cubic>
         <at 0.2011 0.446>
         {arriveTangent=-0.6616, interpMode=RCIM_Cubic, leaveTangent=-0.6616, time=0.4482, value=0}
         <at 0.7469 0.0943>
        ]}
      }})
    Particles.Glow_Rainbow Color:LinearColor=&ColorFromCurve(ColorCurve=DataInterface() {propertyValues {
       AlphaCurve {keys [<at 0 1 RCIM_Cubic>,<at 0.2458 0.3>,<at 1 0 RCIM_Cubic>]}
       BlueCurve {keys [<at 0 0 RCIM_Cubic>, <at 0.214 1>, {arriveTangent=1.4655, interpMode=RCIM_Cubic, leaveTangent=1.4655, time=0.4385, value=1}, <at 0.6823 1>]}
       GreenCurve {keys [
         <at 0 0.0324 RCIM_Cubic>
         <at 0.214 0.2997>
         {arriveTangent=0.8467, interpMode=RCIM_Cubic, leaveTangent=0.8467, time=0.4385, value=0.3445}
         <at 0.6823 0.6101>
        ]}
       RedCurve {keys [
         <at 0 1 RCIM_Cubic>
         <at 0.214 0.669>
         {arriveTangent=-1.2347, interpMode=RCIM_Cubic, leaveTangent=-1.2347, time=0.4385, value=0}
         <at 0.6823 0.1575>
        ]}
      }})
    Particles.Glow_Regular Color:LinearColor=&ColorFromCurve(ColorCurve=DataInterface() {propertyValues {
       AlphaCurve {keys [<at 0 1 RCIM_Cubic>,<at 0.2458 0.3>,<at 1 0 RCIM_Cubic>]}
       BlueCurve {keys [<at 0 0 RCIM_Cubic>,<at 0.2641 1>,<at 0.4304 1 RCIM_Cubic>,<at 0.7211 1>]}
       GreenCurve {keys [
         <at 0 0.0324 RCIM_Cubic>
         <at 0.2641 0.2997>
         {arriveTangent=0.6791, interpMode=RCIM_Cubic, leaveTangent=0.6791, time=0.4304, value=0.3445}
         <at 0.7211 0.6101>
        ]}
       RedCurve {keys [
         <at 0 1 RCIM_Cubic>
         <at 0.2641 0.669>
         {arriveTangent=-1.1191, interpMode=RCIM_Cubic, leaveTangent=-1.1191, time=0.4304, value=0}
         <at 0.7211 0.1575>
        ]}
      }})
    Particles.Release_A Color:LinearColor=&MakeLinearColorFromVectorAndFloat(
     "Vector (RGB)"=&MakeVectorFromLinearColorRGB(Color=&ColorFromCurve(ColorCurve=DataInterface() {propertyValues {
         AlphaCurve {keys []}
         BlueCurve {keys [<at 0.0292 0 RCIM_Cubic>,<at 0.5331 2 RCIM_Cubic>]}
         GreenCurve {keys [<at 0.0292 0 RCIM_Cubic>,<at 0.5331 1.8345 RCIM_Cubic>]}
         RedCurve {keys [<at 0.0292 4 RCIM_Cubic>,<at 0.5331 1.1934 RCIM_Cubic>]}
        }}))
     "Float (Alpha)"=&FloatFromCurve(FloatCurve=DataInterface() {propertyValues {Curve {keys [
          {arriveTangent=63.1994, interpMode=RCIM_Cubic, leaveTangent=63.1994, tangentMode=RCTM_User, time=-0.0023, value=0}
          {interpMode=RCIM_Cubic, leaveTangent=-1.275, tangentMode=RCTM_Break, time=0.0567, value=1}
          {arriveTangent=-1.0601, interpMode=RCIM_Cubic, leaveTangent=-1.0601, time=0.3476, value=0.2706}
          <at 1 0>
         ]}}})
    )
    Particles.Release_B Color:LinearColor=&MakeLinearColorFromVectorAndFloat(
     "Vector (RGB)"=&MakeVectorFromLinearColorRGB(Color=<rgba 0.3639 0.8084 0.8323 0.03>)
     "Float (Alpha)"=&FloatFromCurve(FloatCurve=DataInterface() {propertyValues {Curve {
         defaultValue=1
         keys [{interpMode=RCIM_Cubic, tangentMode=RCTM_User, time=0, value=1}, {interpMode=RCIM_Cubic, tangentMode=RCTM_User, time=1, value=0}]
        }}})
    )
   ) {isSetParameters=true}
   &ScaleMeshSize(Scale Factor=&VectorFromFloat(Value=&FloatFromCurve(
      FloatCurve=DataInterface() {propertyValues {Curve {
         defaultValue=0
         keys [
          {arriveTangent=52.4903, interpMode=RCIM_Cubic, leaveTangent=52.4903, tangentMode=RCTM_User, time=0, value=0}
          {arriveTangent=-24.6613, interpMode=RCIM_Cubic, leaveTangent=0.8982, tangentMode=RCTM_Break, time=0.052, value=0.7784}
          {tangentMode=RCTM_User, time=1, value=1}
         ]
        }}}
      Scale Curve=1.5
     )))
   &SpriteRotationRate(Rotation Rate=&MergeHelper_FloatSwitch(
     0=0
     1=0
     2=0
     3=0
     4=0
     5=&RandomRangeFloat(Minimum=-10, Maximum=10)
     6=&RandomRangeFloat(Minimum=-10, Maximum=10)
     Index=&SpawnGroup()
     Num=::NewEnumerator5
    ))
  ]
  rendererClasses [&NiagaraSpriteRendererProperties, &NiagaraMeshRendererProperties]
  renderers [
   {rendererClass=&NiagaraSpriteRendererProperties}
   {rendererClass=&NiagaraMeshRendererProperties, rendererIndex=1}
   {rendererClass=&NiagaraSpriteRendererProperties, rendererIndex=2}
   {rendererClass=&NiagaraSpriteRendererProperties, rendererIndex=3}
   {rendererClass=&NiagaraSpriteRendererProperties, rendererIndex=4}
   {rendererClass=&NiagaraSpriteRendererProperties, rendererIndex=5}
   {rendererClass=&NiagaraSpriteRendererProperties, rendererIndex=6}
  ]
 }
 Glowy {
  bEnabled=false
  emitterUpdateScript [
   &EmitterState(Life Cycle Mode=::NewEnumerator1, Loop Behavior=::NewEnumerator1, Loop Duration=2, UseLoopDelay=false)
   &SpawnBurst_Instantaneous(Use Spawn Probability=false, UseLoopCountLimit=false) {module=SpawnBurst_Instantaneous002}
   &SpawnBurst_Instantaneous(Spawn Group=3, Use Spawn Probability=false, UseLoopCountLimit=false) {module=SpawnBurst_Instantaneous001}
   &SpawnBurst_Instantaneous(Spawn Group=2, Use Spawn Probability=false, UseLoopCountLimit=false)
  ]
  particleSpawnScript [
   /SetVariables(Particles.VisibilityTag=&SpawnGroup()) {isSetParameters=true}
   &InitializeParticle(
    Color Mode=::NewEnumerator0
    Lifetime=&MergeHelper_FloatSwitch(0=0.8, 1=0.8, 2=0.5, 3=0.8, Index=&SpawnGroup(), Num=::NewEnumerator2)
    Mesh Scale=<0.28 0.4 0.4>
    Mesh Scale Mode=::NewEnumerator4
    Position Offset=&MergeHelper_VectorSwitch(0=<0 0 0>, 1=<0 0 0>, 2=<-18 0 0>, 3=<0 0 0>, Index=&SpawnGroup(), Num=::NewEnumerator2)
    Sprite Rotation Mode=::NewEnumerator2
    Sprite Size Mode=::NewEnumerator3
    Uniform Sprite Size=&MergeHelper_FloatSwitch(0=50, 1=5, 2=55, 3=60, Index=&SpawnGroup(), Num=::NewEnumerator2)
    UsePositionOffset=true
    Write Material Random=true
   )
   &CameraOffset(Camera Offset Amount=-3)
  ]
  particleUpdateScript [
   &ParticleState()
   &ScaleSpriteSize(
    Scale Sprite Size Mode=::NewEnumerator0
    Uniform Scale Factor=&FloatFromCurve(FloatCurve=DataInterface() {propertyValues {Curve {
        defaultValue=0
        keys [
         {arriveTangent=52.4903, interpMode=RCIM_Cubic, leaveTangent=52.4903, tangentMode=RCTM_User, time=0, value=0}
         {arriveTangent=-24.6613, interpMode=RCIM_Cubic, leaveTangent=-5.5991, tangentMode=RCTM_Break, time=0.0549, value=1.1994}
         {arriveTangent=-1.2691, interpMode=RCIM_Cubic, leaveTangent=-1.2691, time=0.2916, value=0.2084}
         {tangentMode=RCTM_User, time=1, value=0}
        ]
       }}})
   )
   &ScaleColor(
    Linear Color Curve=DataInterface() {propertyValues {
      AlphaCurve {keys [<at 0 0>,<at 0.0057 1>,<at 0.9911 1>,<at 1 0>]}
      BlueCurve {keys [<at 0 1>,<at 0.1887 1>,<at 0.4094 0.0035>,<at 0.6627 0>,<at 0.9895 0>]}
      GreenCurve {keys [<at 0 0.7342>,<at 0.1887 0.2865>,<at 0.4094 0.4832>,<at 0.6627 0.0306>,<at 0.9895 0.0167>]}
      RedCurve {keys [<at 0 0>,<at 0.1887 0.4632>,<at 0.4094 1>,<at 0.6627 1>,<at 0.9895 0.5469>]}
     }}
    Scale Mode=::NewEnumerator2
   ) {enabled=false}
   /SetVariables(Particles.color1:LinearColor=&ColorFromCurve(ColorCurve=DataInterface() {propertyValues {
       AlphaCurve {keys [<at 0 1 RCIM_Cubic>,<at 0.2458 0.3>,<at 1 0 RCIM_Cubic>]}
       BlueCurve {keys [<at 0 0 RCIM_Cubic>,<at 0.2641 1>,<at 0.4304 1 RCIM_Cubic>,<at 0.7211 1>]}
       GreenCurve {keys [
         <at 0 0.0324 RCIM_Cubic>
         <at 0.2641 0.2997>
         {arriveTangent=0.6791, interpMode=RCIM_Cubic, leaveTangent=0.6791, time=0.4304, value=0.3445}
         <at 0.7211 0.6101>
        ]}
       RedCurve {keys [
         <at 0 1 RCIM_Cubic>
         <at 0.2641 0.669>
         {arriveTangent=-1.1191, interpMode=RCIM_Cubic, leaveTangent=-1.1191, time=0.4304, value=0}
         <at 0.7211 0.1575>
        ]}
      }})) {isSetParameters=true}
   /SetVariables(Particles.color2:LinearColor=&ColorFromCurve(ColorCurve=DataInterface() {propertyValues {
       AlphaCurve {keys [<at 0 0>,<at 0.0057 1>,<at 0.9911 1>,<at 1 0>]}
       BlueCurve {keys [<at 0 2>,<at 0.1887 2>,<at 0.4094 0.8708>,<at 0.6626 0.8426>,<at 0.8248 0.086>]}
       GreenCurve {keys [<at 0 1.7258>,<at 0.1887 0.7351>,<at 0.4094 1.3024>,<at 0.6626 1.0345>,<at 0.8248 0.0003>]}
       RedCurve {keys [<at 0 1.0222>,<at 0.1887 0.9789>,<at 0.4094 2>,<at 0.6626 2>,<at 0.8248 3>]}
      }})) {enabled=false, isSetParameters=true}
   /SetVariables(Particles.color3:LinearColor=&ColorFromCurve(ColorCurve=DataInterface() {propertyValues {
       AlphaCurve {keys [<at 0 0.7 RCIM_Cubic>,<at 0.1384 0.1>,<at 0.9724 0 RCIM_Cubic>]}
       BlueCurve {keys [
         <at 0 0 RCIM_Cubic>
         <at 0.2011 0.6667>
         {arriveTangent=0.8019, interpMode=RCIM_Cubic, leaveTangent=0.8019, time=0.4482, value=0.6146}
         <at 0.7469 0.599>
        ]}
       GreenCurve {keys [
         <at 0 0.0191 RCIM_Cubic>
         <at 0.2011 0.1998>
         {arriveTangent=0.4637, interpMode=RCIM_Cubic, leaveTangent=0.4637, time=0.4482, value=0.2117}
         <at 0.7469 0.3654>
        ]}
       RedCurve {keys [
         <at 0 0.5885 RCIM_Cubic>
         <at 0.2011 0.446>
         {arriveTangent=-0.6616, interpMode=RCIM_Cubic, leaveTangent=-0.6616, time=0.4482, value=0}
         <at 0.7469 0.0943>
        ]}
      }})) {isSetParameters=true}
   /SetVariables(Particles.color4:LinearColor=&ColorFromCurve(ColorCurve=DataInterface() {propertyValues {
       AlphaCurve {keys [<at 0 1 RCIM_Cubic>,<at 0.2458 0.3>,<at 1 0 RCIM_Cubic>]}
       BlueCurve {keys [<at 0 0 RCIM_Cubic>, <at 0.214 1>, {arriveTangent=1.4655, interpMode=RCIM_Cubic, leaveTangent=1.4655, time=0.4385, value=1}, <at 0.6823 1>]}
       GreenCurve {keys [
         <at 0 0.0324 RCIM_Cubic>
         <at 0.214 0.2997>
         {arriveTangent=0.8467, interpMode=RCIM_Cubic, leaveTangent=0.8467, time=0.4385, value=0.3445}
         <at 0.6823 0.6101>
        ]}
       RedCurve {keys [
         <at 0 1 RCIM_Cubic>
         <at 0.214 0.669>
         {arriveTangent=-1.2347, interpMode=RCIM_Cubic, leaveTangent=-1.2347, time=0.4385, value=0}
         <at 0.6823 0.1575>
        ]}
      }})) {isSetParameters=true}
   &ScaleMeshSize(Scale Factor=&VectorFromFloat(Value=&FloatFromCurve(
      FloatCurve=DataInterface() {propertyValues {Curve {
         defaultValue=0
         keys [
          {arriveTangent=52.4903, interpMode=RCIM_Cubic, leaveTangent=52.4903, tangentMode=RCTM_User, time=0, value=0}
          {arriveTangent=-24.6613, interpMode=RCIM_Cubic, leaveTangent=0.8982, tangentMode=RCTM_Break, time=0.052, value=0.7784}
          {tangentMode=RCTM_User, time=1, value=1}
         ]
        }}}
      Scale Curve=1.5
     )))
   &DynamicMaterialParameters(
    Index 0 Param 1=2
    Index 0 Param 2=1
    Index 0 Param 3=1
    Index 0 Param 4=&FloatFromCurve(FloatCurve=DataInterface() {propertyValues {Curve {keys [
         {arriveTangent=24.6012, interpMode=RCIM_Cubic, leaveTangent=24.6012, tangentMode=RCTM_User, time=0, value=0}
         {interpMode=RCIM_Cubic, tangentMode=RCTM_User, time=0.0673, value=0.25}
         {interpMode=RCIM_Cubic, tangentMode=RCTM_User, time=0.2691, value=0.25}
         {arriveTangent=-25.7513, interpMode=RCIM_Cubic, leaveTangent=-0.7824, tangentMode=RCTM_Break, time=0.3192, value=0}
         {arriveTangent=-0.2454, interpMode=RCIM_Cubic, leaveTangent=29.1486, tangentMode=RCTM_Break, time=0.3459, value=0}
         {interpMode=RCIM_Cubic, tangentMode=RCTM_User, time=0.3919, value=0.25}
         {interpMode=RCIM_Cubic, tangentMode=RCTM_User, time=0.5975, value=0.25}
         {arriveTangent=-32.7966, interpMode=RCIM_Cubic, leaveTangent=-0.1045, tangentMode=RCTM_Break, time=0.6509, value=0}
         {arriveTangent=0.3404, interpMode=RCIM_Cubic, leaveTangent=54.7669, tangentMode=RCTM_Break, time=0.6737, value=0}
         {interpMode=RCIM_Cubic, tangentMode=RCTM_User, time=0.7258, value=0.25}
         {interpMode=RCIM_Cubic, tangentMode=RCTM_User, time=0.9213, value=0.25}
         {arriveTangent=-19.5198, interpMode=RCIM_Cubic, leaveTangent=-19.5198, tangentMode=RCTM_User, time=1, value=0}
        ]}}})
    Module.Index0Param0Write=true
    Module.Index0Param1Write=true
    Module.Index0Param2Write=true
    Module.Index0Param3Write=true
   ) {enabled=false, module=DynamicMaterialParameters001}
   /SetVariables(Particles.hand glow size:Vector2f=&Multiply_Vector2D(
     A=$Particles.Initial.SpriteSize
     B=&Vector2DFromFloat(Value=&FloatFromCurve(
       FloatCurve=DataInterface() {propertyValues {Curve {
          defaultValue=1
          keys [
           {arriveTangent=10.5218, interpMode=RCIM_Cubic, leaveTangent=10.5218, tangentMode=RCTM_User, time=0, value=0}
           {arriveTangentWeight=0.041, interpMode=RCIM_Cubic, leaveTangentWeight=0.2454, tangentMode=RCTM_User, time=0.123, value=1}
           {arriveTangentWeight=0.3722, interpMode=RCIM_Cubic, leaveTangentWeight=0.0711, tangentMode=RCTM_User, time=0.8593, value=1}
           {arriveTangent=-10.1885, tangentMode=RCTM_User, time=1, value=0}
          ]
         }}}
       Scale Curve=0.5
      ))
    )) {enabled=false, isSetParameters=true}
  ]
  rendererClasses [&NiagaraSpriteRendererProperties, &NiagaraMeshRendererProperties]
  renderers [
   {rendererClass=&NiagaraSpriteRendererProperties}
   {rendererClass=&NiagaraMeshRendererProperties, rendererIndex=1}
   {rendererClass=&NiagaraSpriteRendererProperties, rendererIndex=2}
  ]
 }
}
summary {
 systemName=NS_Emote_Sitcom_Burst
 emitters [
  {
   bEnabled=false
   emitterName=Glowy
   rendererClasses [&NiagaraSpriteRendererProperties, &NiagaraMeshRendererProperties]
  }
  {bEnabled=false, emitterName=Calm_Release, rendererClasses [&NiagaraSpriteRendererProperties]}
  {bEnabled=false, emitterName=Additional_Glow, rendererClasses [&NiagaraSpriteRendererProperties]}
  {
   emitterName=Condensed_BurstRelease
   rendererClasses [&NiagaraSpriteRendererProperties, &NiagaraMeshRendererProperties]
  }
 ]
 userVariables [
  User.bIsFrontEnd {type=NiagaraBool}
  User.bIsFrontEndPreview {type=NiagaraBool}
 ]
}
