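# Niagara effect for the Sitcom emote: a chest glow that flashes in three growing stages, shifting from blue through orange to red, with barrier pulses.
# Condensed_ChestGlow merges glow sprite, mesh, build-up and release pulses into one emitter, using spawn groups for timing, scale and renderer visibility tags.
# Per-color particle attributes carry curve-driven fades, with pulse alpha adjusted for gameplay versus front end.
# The separate barrier, Glowy, Little_Lines and Minimal emitters are disabled.
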NiagaraSystem /BRCosmetics/Animation/Game/MainPlayer/Emotes/Sitcom/FX/NS_Emote_Sitcom_Glow
dependencies {
 usedDataInterfaces [
  &NiagaraDataInterfaceSpriteRendererInfo
  &NiagaraDataInterfaceCurve
  &NiagaraDataInterfaceColorCurve
 ]
 usedDynamicInputs []
 usedModules [&SystemState]
}
emitters {
 Barrier_PulseBuildUp {
  bEnabled=false
  emitterUpdateScript [
   &EmitterState(Life Cycle Mode=::NewEnumerator1, Loop Behavior=::NewEnumerator1, Loop Duration=1.834, UseLoopDelay=false)
   &SpawnBurst_Instantaneous(Use Spawn Probability=false, UseLoopCountLimit=false)
   &SpawnBurst_Instantaneous(Spawn Group=1, Spawn Time=0.7, Use Spawn Probability=false, UseLoopCountLimit=false) {module=SpawnBurst_Instantaneous002}
   &SpawnBurst_Instantaneous(Spawn Group=1, Spawn Time=0.45, Use Spawn Probability=false, UseLoopCountLimit=false) {enabled=false, module=SpawnBurst_Instantaneous004}
   &SpawnBurst_Instantaneous(Spawn Group=2, Spawn Time=1.35, Use Spawn Probability=false, UseLoopCountLimit=false) {module=SpawnBurst_Instantaneous001}
   &SpawnBurst_Instantaneous(Spawn Group=2, Spawn Time=1.15, Use Spawn Probability=false, UseLoopCountLimit=false) {enabled=false, module=SpawnBurst_Instantaneous003}
   &SpawnBurst_Instantaneous(Spawn Group=2, Spawn Time=1.5, Use Spawn Probability=false, UseLoopCountLimit=false) {enabled=false, module=SpawnBurst_Instantaneous005}
  ]
  particleSpawnScript [
   /SetVariables(Particles.VisibilityTag=&SpawnGroup()) {isSetParameters=true, module=SetVariables_#1}
   &InitializeParticle(
    Color=&MergeHelper_ColorSwitch(
     0=<rgba 0.2219 0.5513 1 1>
     1=<rgba 1 0.2449 0.0427 1>
     2=<rgba 1 0 0.0162 1>
     3=?
     Index=&SpawnGroup()
     Num=::NewEnumerator1
    )
    Lifetime=0.7
    Mesh Scale Mode=::NewEnumerator3
    Mesh Uniform Scale=&MergeHelper_FloatSwitch(0=0.135, 1=0.135, 2=0.135, 3=0.18, Index=&SpawnGroup(), Num=::NewEnumerator2)
    UsePositionOffset=true
    Write Material Random=true
   )
   &InitialMeshOrientation(Mesh Axis To Orient=<0 1 0>, Mesh Orientation Mode=::NewEnumerator2, Orientation Vector=<0 0 1>, Use Rotation Vector=false)
   /SetVariables(Particles.Linear Color=&RandomRangeLinearColor(Minimum=<rgba 0.3661 0.3904 0.4823 0.05>, Maximum=<rgba 0.759 0.8094 1 0.012>)) {enabled=false, isSetParameters=true, module=SetVariables_#2}
   &CameraOffset(Camera Offset Amount=25)
  ]
  particleUpdateScript [
   &ParticleState()
   &ScaleMeshSize(Scale Factor=&VectorFromFloat(Value=&FloatFromCurve(
      FloatCurve=DataInterface() {propertyValues {Curve {keys [
          {interpMode=RCIM_Cubic, tangentMode=RCTM_User, time=0, value=0}
          {arriveTangent=2, interpMode=RCIM_Cubic, leaveTangent=2, time=0.4676, value=1.5771}
          {interpMode=RCIM_Cubic, tangentMode=RCTM_User, time=1, value=2}
         ]}}}
      Scale Curve=&MergeHelper_FloatSwitch(0=0.5, 1=0.8, 2=1.15, 3=?, Index=&SpawnGroup(), Num=::NewEnumerator1)
     ))) {enabled=false}
   &ScaleMeshSize(Scale Factor=&VectorFromFloat(Value=&FloatFromCurve(
      FloatCurve=DataInterface() {propertyValues {Curve {keys [
          {arriveTangent=110.5934, interpMode=RCIM_Cubic, leaveTangent=110.5934, tangentMode=RCTM_User, time=0, value=0}
          {arriveTangent=-6.0786, interpMode=RCIM_Cubic, leaveTangent=-6.0786, tangentMode=RCTM_User, time=0.0981, value=2.0974}
          {arriveTangent=-0.5219, interpMode=RCIM_Cubic, leaveTangent=-0.5219, tangentMode=RCTM_User, time=1, value=-0.0547}
         ]}}}
      Scale Curve=&MergeHelper_FloatSwitch(0=0.5, 1=0.75, 2=1, 3=?, Index=&SpawnGroup(), Num=::NewEnumerator1)
     ))) {module=ScaleMeshSize001}
   &ScaleColor(
    Scale Alpha=&FloatFromCurve(
     FloatCurve=DataInterface() {propertyValues {Curve {keys [
         {arriveTangent=-3.5042, interpMode=RCIM_Cubic, leaveTangent=-3.5042, tangentMode=RCTM_User, time=0.1, value=1}
         {interpMode=RCIM_Cubic, tangentMode=RCTM_User, time=1, value=0}
        ]}}}
     Scale Curve=&FN_FrontEndFloatModifier(Gameplay=2, Front End=1.125)
    )
    ScaleA=true
    ScaleRGB=true
   )
   &DynamicMaterialParameters(
    Index 0 Param 1=2
    Index 0 Param 2=1
    Index 0 Param 3=1
    Index 0 Param 4=&FloatFromCurve(FloatCurve=DataInterface() {propertyValues {Curve {
        defaultValue=0
        keys [
         {arriveTangent=4.1785, interpMode=RCIM_Cubic, leaveTangent=4.1785, tangentMode=RCTM_User, time=0, value=0}
         {arriveTangent=1.1179, interpMode=RCIM_Cubic, leaveTangent=1.1179, tangentMode=RCTM_User, time=0.2042, value=0.6228}
         {arriveTangent=0.0224, interpMode=RCIM_Cubic, leaveTangent=0.0224, tangentMode=RCTM_User, time=1, value=1}
        ]
       }}})
    Module.Index0Param0Write=true
    Module.Index0Param1Write=true
    Module.Index0Param2Write=true
    Module.Index0Param3Write=true
   ) {module=DynamicMaterialParameters001}
   /SetVariables(Particles.Position=&AddVectorToPosition(
     Position=$Particles.Position
     Vector=&MakeVector(
      X=&FloatFromCurve(
       FloatCurve=DataInterface() {propertyValues {Curve {keys [
           {interpMode=RCIM_Cubic, tangentMode=RCTM_User, time=0, value=1}
           {arriveTangent=-0.6867, interpMode=RCIM_Cubic, leaveTangent=-0.6867, tangentMode=RCTM_User, time=1, value=0}
          ]}}}
       Scale Curve=0.2
      )
      Y=0
      Z=0
     )
    )) {isSetParameters=true, module=SetVariables_#3}
  ]
  rendererClasses [&NiagaraMeshRendererProperties]
  renderers [
   {rendererClass=&NiagaraMeshRendererProperties, rendererIndex=0}
   {rendererClass=&NiagaraMeshRendererProperties, rendererIndex=1}
   {rendererClass=&NiagaraMeshRendererProperties, rendererIndex=2}
  ]
 }
 Barrier_Release {
  bEnabled=false
  emitterUpdateScript [
   &EmitterState(Life Cycle Mode=::NewEnumerator1, Loop Behavior=::NewEnumerator1, Loop Duration=1.834, UseLoopDelay=false)
   &SpawnBurst_Instantaneous(Use Spawn Probability=false, UseLoopCountLimit=false)
   &SpawnBurst_Instantaneous(Spawn Group=1, Spawn Time=0.65, Use Spawn Probability=false, UseLoopCountLimit=false) {module=SpawnBurst_Instantaneous002}
   &SpawnBurst_Instantaneous(Spawn Group=1, Spawn Time=0.75, Use Spawn Probability=false, UseLoopCountLimit=false) {module=SpawnBurst_Instantaneous004}
   &SpawnBurst_Instantaneous(Spawn Group=2, Spawn Time=1.2, Use Spawn Probability=false, UseLoopCountLimit=false) {module=SpawnBurst_Instantaneous001}
   &SpawnBurst_Instantaneous(Spawn Group=2, Spawn Time=1.3, Use Spawn Probability=false, UseLoopCountLimit=false) {module=SpawnBurst_Instantaneous005}
   &SpawnBurst_Instantaneous(Spawn Group=2, Spawn Time=1.4, Use Spawn Probability=false, UseLoopCountLimit=false) {module=SpawnBurst_Instantaneous003}
  ]
  particleSpawnScript [
   /SetVariables(Particles.VisibilityTag=&SpawnGroup()) {enabled=false, isSetParameters=true, module=SetVariables_#1}
   &InitializeParticle(
    Color=&MergeHelper_ColorSwitch(
     0=<rgba 0.2219 0.5513 1 0.6>
     1=<rgba 1 0.2449 0.0427 0.6>
     2=<rgba 1 0 0.0162 0.6>
     3=?
     Index=&SpawnGroup()
     Num=::NewEnumerator1
    )
    Lifetime=0.3
    Mesh Scale Mode=::NewEnumerator3
    Mesh Uniform Scale=&MergeHelper_FloatSwitch(0=0.135, 1=0.135, 2=0.135, 3=0.18, Index=&SpawnGroup(), Num=::NewEnumerator2)
    UsePositionOffset=true
    Write Material Random=true
   )
   &InitialMeshOrientation(Mesh Axis To Orient=<0 1 0>, Mesh Orientation Mode=::NewEnumerator2, Orientation Vector=<0 0 1>, Use Rotation Vector=false)
   /SetVariables(Particles.Linear Color=&RandomRangeLinearColor(Minimum=<rgba 0.3661 0.3904 0.4823 0.05>, Maximum=<rgba 0.759 0.8094 1 0.012>)) {enabled=false, isSetParameters=true, module=SetVariables_#2}
   &CameraOffset(Camera Offset Amount=25)
  ]
  particleUpdateScript [
   &ParticleState()
   &ScaleMeshSize(Scale Factor=&VectorFromFloat(Value=&FloatFromCurve(
      FloatCurve=DataInterface() {propertyValues {Curve {keys [
          {interpMode=RCIM_Cubic, tangentMode=RCTM_User, time=0, value=0}
          {arriveTangent=5.4701, interpMode=RCIM_Cubic, leaveTangent=5.4701, tangentMode=RCTM_User, time=0.6417, value=1.2571}
          {interpMode=RCIM_Cubic, tangentMode=RCTM_User, time=1, value=2}
         ]}}}
      Scale Curve=&MergeHelper_FloatSwitch(0=1.6, 1=2, 2=2.5, 3=?, Index=&SpawnGroup(), Num=::NewEnumerator1)
     )))
   &ScaleColor(
    Scale Alpha=&FloatFromCurve(
     FloatCurve=DataInterface() {propertyValues {Curve {keys [
         {arriveTangent=-3.5042, interpMode=RCIM_Cubic, leaveTangent=-3.5042, tangentMode=RCTM_User, time=0.1, value=1}
         {interpMode=RCIM_Cubic, tangentMode=RCTM_User, time=1, value=0}
        ]}}}
     Scale Curve=&FN_FrontEndFloatModifier(Gameplay=2, Front End=1.125)
    )
    ScaleA=true
    ScaleRGB=true
   )
   &DynamicMaterialParameters(
    Index 0 Param 1=2
    Index 0 Param 2=1
    Index 0 Param 3=1
    Index 0 Param 4=&FloatFromCurve(FloatCurve=DataInterface() {propertyValues {Curve {
        defaultValue=0
        keys [
         {arriveTangent=4.1785, interpMode=RCIM_Cubic, leaveTangent=4.1785, tangentMode=RCTM_User, time=0, value=0}
         {arriveTangent=1.1179, interpMode=RCIM_Cubic, leaveTangent=1.1179, tangentMode=RCTM_User, time=0.2042, value=0.6228}
         {arriveTangent=0.0224, interpMode=RCIM_Cubic, leaveTangent=0.0224, tangentMode=RCTM_User, time=1, value=1}
        ]
       }}})
    Module.Index0Param0Write=true
    Module.Index0Param1Write=true
    Module.Index0Param2Write=true
    Module.Index0Param3Write=true
   ) {module=DynamicMaterialParameters001}
   /SetVariables(Particles.Position=&AddVectorToPosition(
     Position=$Particles.Position
     Vector=&MakeVector(
      X=&FloatFromCurve(
       FloatCurve=DataInterface() {propertyValues {Curve {keys [
           {interpMode=RCIM_Cubic, tangentMode=RCTM_User, time=0, value=1}
           {arriveTangent=-0.6867, interpMode=RCIM_Cubic, leaveTangent=-0.6867, tangentMode=RCTM_User, time=1, value=0}
          ]}}}
       Scale Curve=0.2
      )
      Y=0
      Z=0
     )
    )) {isSetParameters=true, module=SetVariables_#3}
  ]
  rendererClasses [&NiagaraMeshRendererProperties]
  renderers [
   {rendererClass=&NiagaraMeshRendererProperties, rendererIndex=0}
   {rendererClass=&NiagaraMeshRendererProperties, rendererIndex=1}
   {rendererClass=&NiagaraMeshRendererProperties, rendererIndex=2}
  ]
 }
 Condensed_ChestGlow {
  emitterUpdateScript [
   &EmitterState(Life Cycle Mode=::NewEnumerator1, Loop Behavior=::NewEnumerator1, Loop Duration=1.834, UseLoopDelay=false)
   &SpawnBurst_Instantaneous(Use Spawn Probability=false, UseLoopCountLimit=false)
   &SpawnBurst_Instantaneous(Spawn Group=1, Use Spawn Probability=false, UseLoopCountLimit=false) {module=SpawnBurst_Instantaneous001}
   &SpawnBurst_Instantaneous(Spawn Group=2, Use Spawn Probability=false, UseLoopCountLimit=false) {module=SpawnBurst_Instantaneous002}
   &SpawnBurst_Instantaneous(Spawn Group=3, Spawn Time=0.7, Use Spawn Probability=false, UseLoopCountLimit=false) {module=SpawnBurst_Instantaneous003}
   &SpawnBurst_Instantaneous(Spawn Group=4, Spawn Time=1.35, Use Spawn Probability=false, UseLoopCountLimit=false) {module=SpawnBurst_Instantaneous004}
   &SpawnBurst_Instantaneous(Spawn Group=5, Use Spawn Probability=false, UseLoopCountLimit=false) {module=SpawnBurst_Instantaneous010}
   &SpawnBurst_Instantaneous(Spawn Group=6, Spawn Time=0.65, Use Spawn Probability=false, UseLoopCountLimit=false) {module=SpawnBurst_Instantaneous009}
   &SpawnBurst_Instantaneous(Spawn Group=6, Spawn Time=0.75, Use Spawn Probability=false, UseLoopCountLimit=false) {module=SpawnBurst_Instantaneous008}
   &SpawnBurst_Instantaneous(Spawn Group=7, Spawn Time=1.2, Use Spawn Probability=false, UseLoopCountLimit=false) {module=SpawnBurst_Instantaneous007}
   &SpawnBurst_Instantaneous(Spawn Group=7, Spawn Time=1.3, Use Spawn Probability=false, UseLoopCountLimit=false) {module=SpawnBurst_Instantaneous006}
   &SpawnBurst_Instantaneous(Spawn Group=7, Spawn Time=1.4, Use Spawn Probability=false, UseLoopCountLimit=false) {module=SpawnBurst_Instantaneous005}
  ]
  particleSpawnScript [
   /SetVariables(Particles.VisibilityTag=&SpawnGroup()) {isSetParameters=true}
   &InitializeParticle(
    Color Mode=::NewEnumerator0
    Lifetime=&MergeHelper_FloatSwitch(0=1.84, 1=1.84, 2=0.7, 3=0.7, 4=0.7, 5=0.3, 6=0.3, 7=0.3, Index=&SpawnGroup(), Num=::NewEnumerator6)
    Mesh Scale=&MergeHelper_VectorSwitch(
     0=<0 0 0>
     1=<0.28 0.4 0.4>
     2=<0.135 0.135 0.135>
     3=<0.135 0.135 0.135>
     4=<0.135 0.135 0.135>
     5=<0.135 0.135 0.135>
     6=<0.135 0.135 0.135>
     7=<0.135 0.135 0.135>
     Index=&SpawnGroup()
     Num=::NewEnumerator6
    )
    Mesh Scale Mode=::NewEnumerator4
    Position Offset=&MaskVector(Mask Source=0, Mask Parameter=?, Mask=1, Masked Vector=<-14 0 0>, Passthrough Vector=<0 0 0>)
    Sprite Rotation Mode=::NewEnumerator2
    Sprite Size Mode=::NewEnumerator3
    Uniform Sprite Size=42
    UsePositionOffset=true
    Write Material Random=true
   )
   &InitialMeshOrientation(Mesh Axis To Orient=<0 1 0>, Mesh Orientation Mode=::NewEnumerator2, Orientation Vector=<0 0 1>, Use Rotation Vector=false)
   /SetVariables(
    Particles.Blue:LinearColor=<rgba 0.2219 0.5513 1 1>
    Particles.Orange:LinearColor=<rgba 1 0.2449 0.0427 1>
    Particles.Red:LinearColor=<rgba 1 0 0.0162 1>
   ) {isSetParameters=true}
   &CameraOffset(Camera Offset Amount=&MergeHelper_FloatSwitch(0=-3, 1=2, 2=25, 3=25, 4=25, 5=25, 6=25, 7=25, Index=&SpawnGroup(), Num=::NewEnumerator6))
   /SetVariables(Particles.Linear Color=<rgba 1 1 1 0.2>) {isSetParameters=true}
  ]
  particleUpdateScript [
   &ParticleState()
   /SetVariables(
    Particles.Glow_Mesh_Color:LinearColor=&ColorFromCurve(ColorCurve=DataInterface() {propertyValues {
       AlphaCurve {keys [<at 0 0>,<at 0.0057 0.3>,<at 0.9911 0.3>,<at 1 0>]}
       BlueCurve {keys [<at 0 0.5281>,<at 0.1887 0.4156>,<at 0.4094 0.0286>,<at 0.6626 0>,<at 0.9895 0>]}
       GreenCurve {keys [<at 0 0.3878>,<at 0.1887 0.119>,<at 0.4094 0.2072>,<at 0.6626 0.014>,<at 0.9895 0.0113>]}
       RedCurve {keys [<at 0 0>,<at 0.1887 0.1221>,<at 0.4094 0.4781>,<at 0.6626 0.4573>,<at 0.9895 0.3698>]}
      }})
    Particles.Glow_Rainbow Color:LinearColor=&ColorFromCurve(ColorCurve=DataInterface() {propertyValues {
       AlphaCurve {keys [<at 0 0>,<at 0.0057 0.4>,<at 0.9911 0.4>,<at 1 0>]}
       BlueCurve {keys [<at 0.0831 1>,<at 0.1887 1>,<at 0.4094 0.2394>,<at 0.6626 0.1395>,<at 0.9895 0.015>]}
       GreenCurve {keys [<at 0.0831 0.6362>,<at 0.1887 0.5536>,<at 0.4094 0.5975>,<at 0.6626 0.1681>,<at 0.9895 0.0225>]}
       RedCurve {keys [<at 0.0831 0.3427>,<at 0.1887 0.6608>,<at 0.4094 1>,<at 0.6626 1>,<at 0.9895 1>]}
      }})
    Particles.Scale_Blue:LinearColor=&MakeLinearColorFromVectorAndFloat(
     "Vector (RGB)"=&MakeVectorFromLinearColorRGB(Color=$Particles.Initial.Blue)
     "Float (Alpha)"=&FloatFromCurve(
      FloatCurve=DataInterface() {propertyValues {Curve {keys [
          {arriveTangent=-3.5042, interpMode=RCIM_Cubic, leaveTangent=-3.5042, tangentMode=RCTM_User, time=0.1, value=1}
          {interpMode=RCIM_Cubic, tangentMode=RCTM_User, time=1, value=0}
         ]}}}
      Scale Curve=&FN_FrontEndFloatModifier(Gameplay=4, Front End=2.25)
     )
    )
    Particles.Scale_Blue BuildUp:LinearColor=&MakeLinearColorFromVectorAndFloat(
     "Vector (RGB)"=&MakeVectorFromLinearColorRGB(Color=$Particles.Initial.Blue)
     "Float (Alpha)"=&FloatFromCurve(
      FloatCurve=DataInterface() {propertyValues {Curve {keys [
          {arriveTangent=-3.5042, interpMode=RCIM_Cubic, leaveTangent=-3.5042, tangentMode=RCTM_User, time=0.1, value=1}
          {interpMode=RCIM_Cubic, tangentMode=RCTM_User, time=1, value=0}
         ]}}}
      Scale Curve=&FN_FrontEndFloatModifier(Gameplay=2.2, Front End=1.6)
     )
    )
    Particles.Scale_Orange:LinearColor=&MakeLinearColorFromVectorAndFloat(
     "Vector (RGB)"=&MakeVectorFromLinearColorRGB(Color=$Particles.Initial.Orange)
     "Float (Alpha)"=&FloatFromCurve(
      FloatCurve=DataInterface() {propertyValues {Curve {keys [
          {arriveTangent=-3.5042, interpMode=RCIM_Cubic, leaveTangent=-3.5042, tangentMode=RCTM_User, time=0.1, value=1}
          {interpMode=RCIM_Cubic, tangentMode=RCTM_User, time=1, value=0}
         ]}}}
      Scale Curve=&FN_FrontEndFloatModifier(Gameplay=4, Front End=2.25)
     )
    )
    Particles.Scale_Red:LinearColor=&MakeLinearColorFromVectorAndFloat(
     "Vector (RGB)"=&MakeVectorFromLinearColorRGB(Color=$Particles.Initial.Red)
     "Float (Alpha)"=&FloatFromCurve(
      FloatCurve=DataInterface() {propertyValues {Curve {keys [
          {arriveTangent=-3.5042, interpMode=RCIM_Cubic, leaveTangent=-3.5042, tangentMode=RCTM_User, time=0.1, value=1}
          {interpMode=RCIM_Cubic, tangentMode=RCTM_User, time=1, value=0}
         ]}}}
      Scale Curve=&FN_FrontEndFloatModifier(Gameplay=4, Front End=2.25)
     )
    )
   ) {isSetParameters=true}
   &ScaleSpriteSize(Uniform Curve Sprite Scale=DataInterface() {propertyValues {Curve {keys [
        {arriveTangent=18.0975, interpMode=RCIM_Cubic, leaveTangent=18.0975, tangentMode=RCTM_User, time=0, value=0}
        {arriveTangent=-16.2875, interpMode=RCIM_Cubic, tangentMode=RCTM_Break, time=0.0452, value=0.4914}
        {interpMode=RCIM_Cubic, tangentMode=RCTM_User, time=0.2703, value=0.4936}
        {arriveTangent=-15.132, interpMode=RCIM_Cubic, leaveTangent=-0.7824, tangentMode=RCTM_Break, time=0.3192, value=0}
        {arriveTangent=-0.2454, interpMode=RCIM_Cubic, leaveTangent=29.1486, tangentMode=RCTM_Break, time=0.3459, value=0}
        {arriveTangent=-23.7709, interpMode=RCIM_Cubic, tangentMode=RCTM_Break, time=0.3938, value=0.8436}
        {interpMode=RCIM_Cubic, tangentMode=RCTM_User, time=0.5994, value=0.8544}
        {arriveTangent=-32.7966, interpMode=RCIM_Cubic, leaveTangent=-0.1045, tangentMode=RCTM_Break, time=0.6509, value=0}
        {arriveTangent=0.3404, interpMode=RCIM_Cubic, leaveTangent=54.7669, tangentMode=RCTM_Break, time=0.6737, value=0}
        {arriveTangent=-23.5308, interpMode=RCIM_Cubic, tangentMode=RCTM_Break, time=0.7258, value=1.2653}
        {interpMode=RCIM_Cubic, tangentMode=RCTM_User, time=0.9213, value=1.2653}
        {arriveTangent=-19.5198, interpMode=RCIM_Cubic, leaveTangent=-19.5198, tangentMode=RCTM_User, time=1, value=0}
       ]}}})
   &ScaleMeshSize(Scale Factor=&MergeHelper_VectorSwitch(
     0=<0 0 0>
     1=&VectorFromFloat(Value=&FloatFromCurve(FloatCurve=DataInterface() {propertyValues {Curve {keys [
           {interpMode=RCIM_Cubic, leaveTangentWeight=0.5521, tangentMode=RCTM_User, time=0, value=0}
           {
            arriveTangent=31.0966
            arriveTangentWeight=1.1373
            interpMode=RCIM_Cubic
            leaveTangentWeight=3.4117
            tangentMode=RCTM_Break
            time=0.0673
            value=0.6947
           }
           {interpMode=RCIM_Cubic, leaveTangent=-23.1617, tangentMode=RCTM_Break, time=0.2691, value=0.6968}
           {arriveTangentWeight=0.4308, interpMode=RCIM_Cubic, leaveTangentWeight=0.0113, tangentMode=RCTM_Break, time=0.3192, value=0}
           {arriveTangentWeight=0.0092, interpMode=RCIM_Cubic, leaveTangentWeight=0.4475, tangentMode=RCTM_Break, time=0.3459, value=0}
           {arriveTangent=43.4537, interpMode=RCIM_Cubic, tangentMode=RCTM_Break, time=0.3919, value=0.9432}
           {interpMode=RCIM_Cubic, leaveTangent=-36.9429, tangentMode=RCTM_Break, time=0.5975, value=0.954}
           {arriveTangentWeight=0.5841, interpMode=RCIM_Cubic, leaveTangentWeight=0.0077, tangentMode=RCTM_Break, time=0.6509, value=0}
           {arriveTangentWeight=0.0081, interpMode=RCIM_Cubic, leaveTangentWeight=0.9512, tangentMode=RCTM_Break, time=0.6737, value=0}
           {arriveTangent=67.1689, interpMode=RCIM_Cubic, tangentMode=RCTM_Break, time=0.7258, value=1.2653}
           {interpMode=RCIM_Cubic, leaveTangent=-46.8783, tangentMode=RCTM_Break, time=0.9213, value=1.2653}
           {arriveTangentWeight=0.5126, interpMode=RCIM_Cubic, tangentMode=RCTM_User, time=1, value=0}
          ]}}}))
     2=&VectorFromFloat(Value=&FloatFromCurve(
       FloatCurve=DataInterface() {propertyValues {Curve {keys [
           {arriveTangent=110.5934, interpMode=RCIM_Cubic, leaveTangent=110.5934, tangentMode=RCTM_User, time=0, value=0}
           {arriveTangent=-6.0786, interpMode=RCIM_Cubic, leaveTangent=-6.0786, tangentMode=RCTM_User, time=0.0981, value=2.0974}
           {arriveTangent=-0.5219, interpMode=RCIM_Cubic, leaveTangent=-0.5219, tangentMode=RCTM_User, time=1, value=-0.0547}
          ]}}}
       Scale Curve=0.5
      ))
     3=&VectorFromFloat(Value=&FloatFromCurve(
       FloatCurve=DataInterface() {propertyValues {Curve {keys [
           {arriveTangent=110.5934, interpMode=RCIM_Cubic, leaveTangent=110.5934, tangentMode=RCTM_User, time=0, value=0}
           {arriveTangent=-6.0786, interpMode=RCIM_Cubic, leaveTangent=-6.0786, tangentMode=RCTM_User, time=0.0981, value=2.0974}
           {arriveTangent=-0.5219, interpMode=RCIM_Cubic, leaveTangent=-0.5219, tangentMode=RCTM_User, time=1, value=-0.0547}
          ]}}}
       Scale Curve=0.75
      ))
     4=&VectorFromFloat(Value=&FloatFromCurve(FloatCurve=DataInterface() {propertyValues {Curve {keys [
           {arriveTangent=110.5934, interpMode=RCIM_Cubic, leaveTangent=110.5934, tangentMode=RCTM_User, time=0, value=0}
           {arriveTangent=-6.0786, interpMode=RCIM_Cubic, leaveTangent=-6.0786, tangentMode=RCTM_User, time=0.0981, value=2.0974}
           {arriveTangent=-0.5219, interpMode=RCIM_Cubic, leaveTangent=-0.5219, tangentMode=RCTM_User, time=1, value=-0.0547}
          ]}}}))
     5=&VectorFromFloat(Value=&FloatFromCurve(
       FloatCurve=DataInterface() {propertyValues {Curve {keys [
           {interpMode=RCIM_Cubic, tangentMode=RCTM_User, time=0, value=0}
           {arriveTangent=5.4701, interpMode=RCIM_Cubic, leaveTangent=5.4701, tangentMode=RCTM_User, time=0.6417, value=1.2571}
           {interpMode=RCIM_Cubic, tangentMode=RCTM_User, time=1, value=2}
          ]}}}
       Scale Curve=1.6
      ))
     6=&VectorFromFloat(Value=&FloatFromCurve(
       FloatCurve=DataInterface() {propertyValues {Curve {keys [
           {interpMode=RCIM_Cubic, tangentMode=RCTM_User, time=0, value=0}
           {arriveTangent=5.4701, interpMode=RCIM_Cubic, leaveTangent=5.4701, tangentMode=RCTM_User, time=0.6417, value=1.2571}
           {interpMode=RCIM_Cubic, tangentMode=RCTM_User, time=1, value=2}
          ]}}}
       Scale Curve=2
      ))
     7=&VectorFromFloat(Value=&FloatFromCurve(
       FloatCurve=DataInterface() {propertyValues {Curve {keys [
           {interpMode=RCIM_Cubic, tangentMode=RCTM_User, time=0, value=0}
           {arriveTangent=5.4701, interpMode=RCIM_Cubic, leaveTangent=5.4701, tangentMode=RCTM_User, time=0.6417, value=1.2571}
           {interpMode=RCIM_Cubic, tangentMode=RCTM_User, time=1, value=2}
          ]}}}
       Scale Curve=2.5
      ))
     Index=&SpawnGroup()
     Num=::NewEnumerator6
    ))
   /SetVariables(Particles.Barrier Dynamic Param:Vector4f=&MakeVector4(
     X=2
     Y=1
     Z=1
     W=&FloatFromCurve(FloatCurve=DataInterface() {propertyValues {Curve {
         defaultValue=0
         keys [
          {arriveTangent=4.1785, interpMode=RCIM_Cubic, leaveTangent=4.1785, tangentMode=RCTM_User, time=0, value=0}
          {arriveTangent=1.1179, interpMode=RCIM_Cubic, leaveTangent=1.1179, tangentMode=RCTM_User, time=0.2042, value=0.6228}
          {arriveTangent=0.0224, interpMode=RCIM_Cubic, leaveTangent=0.0224, tangentMode=RCTM_User, time=1, value=1}
         ]
        }}})
    )) {isSetParameters=true}
   /SetVariables(Particles.Position=&AddVectorToPosition(
     Position=$Particles.Position
     Vector=&MakeVector(
      X=&MergeHelper_FloatSwitch(
       0=0
       1=0
       2=&FloatFromCurve(
        FloatCurve=DataInterface() {propertyValues {Curve {keys [
            {interpMode=RCIM_Cubic, tangentMode=RCTM_User, time=0, value=1}
            {arriveTangent=-0.6867, interpMode=RCIM_Cubic, leaveTangent=-0.6867, tangentMode=RCTM_User, time=1, value=0}
           ]}}}
        Scale Curve=0.2
       )
       3=&FloatFromCurve(
        FloatCurve=DataInterface() {propertyValues {Curve {keys [
            {interpMode=RCIM_Cubic, tangentMode=RCTM_User, time=0, value=1}
            {arriveTangent=-0.6867, interpMode=RCIM_Cubic, leaveTangent=-0.6867, tangentMode=RCTM_User, time=1, value=0}
           ]}}}
        Scale Curve=0.2
       )
       4=&FloatFromCurve(
        FloatCurve=DataInterface() {propertyValues {Curve {keys [
            {interpMode=RCIM_Cubic, tangentMode=RCTM_User, time=0, value=1}
            {arriveTangent=-0.6867, interpMode=RCIM_Cubic, leaveTangent=-0.6867, tangentMode=RCTM_User, time=1, value=0}
           ]}}}
        Scale Curve=0.2
       )
       5=&FloatFromCurve(
        FloatCurve=DataInterface() {propertyValues {Curve {keys [
            {interpMode=RCIM_Cubic, tangentMode=RCTM_User, time=0, value=1}
            {arriveTangent=-0.6867, interpMode=RCIM_Cubic, leaveTangent=-0.6867, tangentMode=RCTM_User, time=1, value=0}
           ]}}}
        Scale Curve=0.2
       )
       Index=&SpawnGroup()
       Num=::NewEnumerator4
      )
      Y=0
      Z=0
     )
    )) {isSetParameters=true}
  ]
  rendererClasses [&NiagaraSpriteRendererProperties, &NiagaraMeshRendererProperties]
  renderers [
   {rendererClass=&NiagaraSpriteRendererProperties}
   {rendererClass=&NiagaraMeshRendererProperties, rendererIndex=1}
   {rendererClass=&NiagaraMeshRendererProperties, rendererIndex=2}
   {rendererClass=&NiagaraMeshRendererProperties, rendererIndex=3}
   {rendererClass=&NiagaraMeshRendererProperties, rendererIndex=4}
   {rendererClass=&NiagaraMeshRendererProperties, rendererIndex=5}
   {rendererClass=&NiagaraMeshRendererProperties, rendererIndex=6}
   {rendererClass=&NiagaraMeshRendererProperties, rendererIndex=7}
  ]
 }
 Glowy {
  bEnabled=false
  emitterUpdateScript [
   &EmitterState(Life Cycle Mode=::NewEnumerator1, Loop Behavior=::NewEnumerator1, Loop Duration=2, UseLoopDelay=false)
   &SpawnBurst_Instantaneous(Use Spawn Probability=false, UseLoopCountLimit=false)
   &SpawnBurst_Instantaneous(Spawn Group=2, Use Spawn Probability=false, UseLoopCountLimit=false) {module=SpawnBurst_Instantaneous002}
   &SpawnBurst_Instantaneous(Spawn Group=3, Use Spawn Probability=false, UseLoopCountLimit=false) {module=SpawnBurst_Instantaneous001}
  ]
  particleSpawnScript [
   /SetVariables(Particles.VisibilityTag=&SpawnGroup()) {isSetParameters=true}
   &InitializeParticle(
    Color Mode=::NewEnumerator0
    Lifetime=1.84
    Mesh Scale=<0.28 0.4 0.4>
    Mesh Scale Mode=::NewEnumerator4
    Position Offset=&MergeHelper_VectorSwitch(0=<0 0 0>, 1=<0 0 0>, 2=<-14 0 0>, 3=<0 0 0>, Index=&SpawnGroup(), Num=::NewEnumerator2)
    Sprite Rotation Mode=::NewEnumerator2
    Sprite Size Mode=::NewEnumerator3
    Uniform Sprite Size=&MergeHelper_FloatSwitch(0=45, 1=4.5, 2=58.5, 3=42, Index=&SpawnGroup(), Num=::NewEnumerator2)
    UsePositionOffset=true
    Write Material Random=true
   )
   &CameraOffset(Camera Offset Amount=&MergeHelper_FloatSwitch(0=0, 1=0, 2=2, 3=-3, Index=&SpawnGroup(), Num=::NewEnumerator2))
  ]
  particleUpdateScript [
   &ParticleState()
   &ScaleSpriteSize(
    Scale Sprite Size Mode=::NewEnumerator0
    Uniform Scale Factor=&FloatFromCurve(FloatCurve=DataInterface() {propertyValues {Curve {keys [
         {arriveTangent=18.0975, interpMode=RCIM_Cubic, leaveTangent=18.0975, tangentMode=RCTM_User, time=0, value=0}
         {arriveTangent=-16.2875, interpMode=RCIM_Cubic, tangentMode=RCTM_Break, time=0.0452, value=0.4914}
         {interpMode=RCIM_Cubic, tangentMode=RCTM_User, time=0.2703, value=0.4936}
         {arriveTangent=-15.132, interpMode=RCIM_Cubic, leaveTangent=-0.7824, tangentMode=RCTM_Break, time=0.3192, value=0}
         {arriveTangent=-0.2454, interpMode=RCIM_Cubic, leaveTangent=29.1486, tangentMode=RCTM_Break, time=0.3459, value=0}
         {arriveTangent=-23.7709, interpMode=RCIM_Cubic, tangentMode=RCTM_Break, time=0.3938, value=0.8436}
         {interpMode=RCIM_Cubic, tangentMode=RCTM_User, time=0.5994, value=0.8544}
         {arriveTangent=-32.7966, interpMode=RCIM_Cubic, leaveTangent=-0.1045, tangentMode=RCTM_Break, time=0.6509, value=0}
         {arriveTangent=0.3404, interpMode=RCIM_Cubic, leaveTangent=54.7669, tangentMode=RCTM_Break, time=0.6737, value=0}
         {arriveTangent=-23.5308, interpMode=RCIM_Cubic, tangentMode=RCTM_Break, time=0.7258, value=1.2653}
         {interpMode=RCIM_Cubic, tangentMode=RCTM_User, time=0.9213, value=1.2653}
         {arriveTangent=-19.5198, interpMode=RCIM_Cubic, leaveTangent=-19.5198, tangentMode=RCTM_User, time=1, value=0}
        ]}}})
   )
   &ScaleColor(
    Linear Color Curve=DataInterface() {propertyValues {
      AlphaCurve {keys [<at 0 0>,<at 0.0057 1>,<at 0.9911 1>,<at 1 0>]}
      BlueCurve {keys [<at 0 1>,<at 0.1887 1>,<at 0.4094 0.0035>,<at 0.6627 0>,<at 0.9895 0>]}
      GreenCurve {keys [<at 0 0.7342>,<at 0.1887 0.2865>,<at 0.4094 0.4832>,<at 0.6627 0.0306>,<at 0.9895 0.0167>]}
      RedCurve {keys [<at 0 0>,<at 0.1887 0.4632>,<at 0.4094 1>,<at 0.6627 1>,<at 0.9895 0.5469>]}
     }}
    Scale Mode=::NewEnumerator2
   ) {enabled=false}
   /SetVariables(Particles.color1:LinearColor=&ColorFromCurve(ColorCurve=DataInterface() {propertyValues {
       AlphaCurve {keys [<at 0 0>,<at 0.0057 1>,<at 0.9911 1>,<at 1 0>]}
       BlueCurve {keys [<at 0.073 1>,<at 0.2207 1>,<at 0.4094 0.0035>,<at 0.6838 0.0521>,<at 1 0>]}
       GreenCurve {keys [<at 0.073 0.7342>,<at 0.2207 0.2865>,<at 0.4094 0.4832>,<at 0.6838 0.0811>,<at 1 0.0306>]}
       RedCurve {keys [<at 0.073 0>,<at 0.2207 0.4632>,<at 0.4094 1>,<at 0.6838 1>,<at 1 1>]}
      }})) {isSetParameters=true}
   /SetVariables(Particles.color2:LinearColor=&ColorFromCurve(ColorCurve=DataInterface() {propertyValues {
       AlphaCurve {keys [<at 0 0>,<at 0.0057 1>,<at 0.9911 1>,<at 1 0>]}
       BlueCurve {keys [<at 0 2>,<at 0.1887 2>,<at 0.4094 0.8708>,<at 0.6626 0.8426>,<at 0.8248 0.086>]}
       GreenCurve {keys [<at 0 1.7258>,<at 0.1887 0.7351>,<at 0.4094 1.3024>,<at 0.6626 1.0345>,<at 0.8248 0.0003>]}
       RedCurve {keys [<at 0 1.0222>,<at 0.1887 0.9789>,<at 0.4094 2>,<at 0.6626 2>,<at 0.8248 3>]}
      }})) {isSetParameters=true}
   /SetVariables(Particles.color3:LinearColor=&ColorFromCurve(ColorCurve=DataInterface() {propertyValues {
       AlphaCurve {keys [<at 0 0>,<at 0.0057 0.3>,<at 0.9911 0.3>,<at 1 0>]}
       BlueCurve {keys [<at 0 0.5281>,<at 0.1887 0.4156>,<at 0.4094 0.0286>,<at 0.6626 0>,<at 0.9895 0>]}
       GreenCurve {keys [<at 0 0.3878>,<at 0.1887 0.119>,<at 0.4094 0.2072>,<at 0.6626 0.014>,<at 0.9895 0.0113>]}
       RedCurve {keys [<at 0 0>,<at 0.1887 0.1221>,<at 0.4094 0.4781>,<at 0.6626 0.4573>,<at 0.9895 0.3698>]}
      }})) {isSetParameters=true}
   /SetVariables(Particles.color4:LinearColor=&ColorFromCurve(ColorCurve=DataInterface() {propertyValues {
       AlphaCurve {keys [<at 0 0>,<at 0.0057 0.4>,<at 0.9911 0.4>,<at 1 0>]}
       BlueCurve {keys [<at 0.0831 1>,<at 0.1887 1>,<at 0.4094 0.2394>,<at 0.6626 0.1395>,<at 0.9895 0.015>]}
       GreenCurve {keys [<at 0.0831 0.6362>,<at 0.1887 0.5536>,<at 0.4094 0.5975>,<at 0.6626 0.1681>,<at 0.9895 0.0225>]}
       RedCurve {keys [<at 0.0831 0.3427>,<at 0.1887 0.6608>,<at 0.4094 1>,<at 0.6626 1>,<at 0.9895 1>]}
      }})) {isSetParameters=true}
   &ScaleMeshSize(Scale Factor=&VectorFromFloat(Value=&FloatFromCurve(FloatCurve=DataInterface() {propertyValues {Curve {keys [
          {interpMode=RCIM_Cubic, leaveTangentWeight=0.5521, tangentMode=RCTM_User, time=0, value=0}
          {
           arriveTangent=31.0966
           arriveTangentWeight=1.1373
           interpMode=RCIM_Cubic
           leaveTangentWeight=3.4117
           tangentMode=RCTM_Break
           time=0.0673
           value=0.6947
          }
          {interpMode=RCIM_Cubic, leaveTangent=-23.1617, tangentMode=RCTM_Break, time=0.2691, value=0.6968}
          {arriveTangentWeight=0.4308, interpMode=RCIM_Cubic, leaveTangentWeight=0.0113, tangentMode=RCTM_Break, time=0.3192, value=0}
          {arriveTangentWeight=0.0092, interpMode=RCIM_Cubic, leaveTangentWeight=0.4475, tangentMode=RCTM_Break, time=0.3459, value=0}
          {arriveTangent=43.4537, interpMode=RCIM_Cubic, tangentMode=RCTM_Break, time=0.3919, value=0.9432}
          {interpMode=RCIM_Cubic, leaveTangent=-36.9429, tangentMode=RCTM_Break, time=0.5975, value=0.954}
          {arriveTangentWeight=0.5841, interpMode=RCIM_Cubic, leaveTangentWeight=0.0077, tangentMode=RCTM_Break, time=0.6509, value=0}
          {arriveTangentWeight=0.0081, interpMode=RCIM_Cubic, leaveTangentWeight=0.9511, tangentMode=RCTM_Break, time=0.6737, value=0}
          {arriveTangent=67.1689, interpMode=RCIM_Cubic, tangentMode=RCTM_Break, time=0.7258, value=1.2653}
          {interpMode=RCIM_Cubic, leaveTangent=-46.8783, tangentMode=RCTM_Break, time=0.9213, value=1.2653}
          {arriveTangentWeight=0.5126, interpMode=RCIM_Cubic, tangentMode=RCTM_User, time=1, value=0}
         ]}}})))
   &DynamicMaterialParameters(
    Index 0 Param 1=2
    Index 0 Param 2=1
    Index 0 Param 3=1
    Index 0 Param 4=&FloatFromCurve(FloatCurve=DataInterface() {propertyValues {Curve {keys [
         {arriveTangent=24.6012, interpMode=RCIM_Cubic, leaveTangent=24.6012, tangentMode=RCTM_User, time=0, value=0}
         {interpMode=RCIM_Cubic, tangentMode=RCTM_User, time=0.0673, value=0.25}
         {interpMode=RCIM_Cubic, tangentMode=RCTM_User, time=0.2691, value=0.25}
         {arriveTangent=-25.7513, interpMode=RCIM_Cubic, leaveTangent=-0.7824, tangentMode=RCTM_Break, time=0.3192, value=0}
         {arriveTangent=-0.2454, interpMode=RCIM_Cubic, leaveTangent=29.1486, tangentMode=RCTM_Break, time=0.3459, value=0}
         {interpMode=RCIM_Cubic, tangentMode=RCTM_User, time=0.3919, value=0.25}
         {interpMode=RCIM_Cubic, tangentMode=RCTM_User, time=0.5975, value=0.25}
         {arriveTangent=-32.7966, interpMode=RCIM_Cubic, leaveTangent=-0.1045, tangentMode=RCTM_Break, time=0.6509, value=0}
         {arriveTangent=0.3404, interpMode=RCIM_Cubic, leaveTangent=54.7669, tangentMode=RCTM_Break, time=0.6737, value=0}
         {interpMode=RCIM_Cubic, tangentMode=RCTM_User, time=0.7258, value=0.25}
         {interpMode=RCIM_Cubic, tangentMode=RCTM_User, time=0.9213, value=0.25}
         {arriveTangent=-19.5198, interpMode=RCIM_Cubic, leaveTangent=-19.5198, tangentMode=RCTM_User, time=1, value=0}
        ]}}})
    Module.Index0Param0Write=true
    Module.Index0Param1Write=true
    Module.Index0Param2Write=true
    Module.Index0Param3Write=true
   ) {module=DynamicMaterialParameters001}
   /SetVariables(Particles.hand glow size:Vector2f=&Multiply_Vector2D(
     A=$Particles.Initial.SpriteSize
     B=&Vector2DFromFloat(Value=&FloatFromCurve(
       FloatCurve=DataInterface() {propertyValues {Curve {
          defaultValue=1
          keys [
           {arriveTangent=10.5218, interpMode=RCIM_Cubic, leaveTangent=10.5218, tangentMode=RCTM_User, time=0, value=0}
           {arriveTangentWeight=0.041, interpMode=RCIM_Cubic, leaveTangentWeight=0.2454, tangentMode=RCTM_User, time=0.123, value=1}
           {arriveTangentWeight=0.3722, interpMode=RCIM_Cubic, leaveTangentWeight=0.0711, tangentMode=RCTM_User, time=0.8593, value=1}
           {arriveTangent=-10.1885, tangentMode=RCTM_User, time=1, value=0}
          ]
         }}}
       Scale Curve=0.5
      ))
    )) {enabled=false, isSetParameters=true}
  ]
  rendererClasses [&NiagaraSpriteRendererProperties, &NiagaraMeshRendererProperties]
  renderers [
   {rendererClass=&NiagaraSpriteRendererProperties}
   {rendererClass=&NiagaraSpriteRendererProperties, rendererIndex=1}
   {rendererClass=&NiagaraSpriteRendererProperties, rendererIndex=2}
   {rendererClass=&NiagaraMeshRendererProperties, rendererIndex=3}
  ]
 }
 Little_Lines {
  bEnabled=false
  emitterUpdateScript [
   &EmitterState()
   &SpawnBurst_Instantaneous(
    Spawn Count=&RandomRangeInt(Minimum=4, Maximum=7)
    Use Spawn Probability=false
    UseLoopCountLimit=false
   )
   &SpawnRate(SpawnRate=400, Use Spawn Probability=false) {enabled=false}
  ]
  particleSpawnScript [
   &InitializeParticle(
    Color=&RandomRangeLinearColor(Minimum=<rgba 0.3288 0.3419 0.4226 0.3>, Maximum=<rgba 0.4031 0.4031 0.4991 0.42>)
    Lifetime Max=0.34
    Lifetime Min=0.2
    Lifetime Mode=::NewEnumerator1
    Sprite Size Max=<1 1.6>
    Sprite Size Min=<0.75 1.1>
    Sprite Size Mode=::NewEnumerator2
    UsePositionOffset=true
    Write Material Random=true
   ) {module=InitializeParticle001}
   &AddVelocity(Velocity=&RandomRangeVector(Minimum=<0 15 0>, Maximum=<0 30 0>, Evaluation Type=::NewEnumerator1))
   &ShapeLocation(
    Handle Radius=0.03
    Large Radius=1
    MaskSpawnGroup=false
    Override Seed=false
    Rotation Axis=<0 1 0>
    Rotation Mode=::NewEnumerator1
    Shape Primitive=::NewEnumerator3
    UseOwnerScale=true
   )
   &AddVelocity(Module.Use Falloff Distance=false, Override Seed=false, Velocity Mode=::NewEnumerator1, Velocity Speed=30) {module=AddVelocity001}
  ]
  particleUpdateScript [
   &ParticleState()
   &ScaleSpriteSize(Scale Sprite Size Mode=::NewEnumerator0)
   &SolveForcesAndVelocity(Clamp Velocity=false, Limit Acceleration=false)
   &ScaleSpriteSizeBySpeed(
    Max Scale Factor=<0.4 10>
    Min Scale Factor=<0.25 5>
    Source Velocity:Vector3f=&Multiply_Vector(A=<0.05 0.05 1>, B=$Particles.Velocity)
    Velocity Threshold=500
   )
   &ScaleColor(
    Scale Alpha=&FloatFromCurve(FloatCurve=DataInterface() {propertyValues {Curve {keys [
         {interpMode=RCIM_Cubic, tangentMode=RCTM_User, time=-0.0004, value=0}
         {interpMode=RCIM_Cubic, tangentMode=RCTM_User, time=0.02, value=1}
         {interpMode=RCIM_Cubic, tangentMode=RCTM_User, time=0.7054, value=1}
         {arriveTangent=-3.76, interpMode=RCIM_Cubic, leaveTangent=-3.76, tangentMode=RCTM_User, time=0.8963, value=0.6518}
         {interpMode=RCIM_Cubic, tangentMode=RCTM_User, time=0.9949, value=0}
        ]}}})
    ScaleA=true
    ScaleRGB=true
   )
  ]
  rendererClasses [&NiagaraSpriteRendererProperties]
  renderers [{rendererClass=&NiagaraSpriteRendererProperties}]
 }
 Minimal {
  bEnabled=false
  emitterUpdateScript [
   &EmitterState(Life Cycle Mode=::NewEnumerator1, Loop Behavior=::NewEnumerator1, Loop Duration=2, UseLoopDelay=false)
   &SpawnBurst_Instantaneous(Spawn Count=3, Use Spawn Probability=false, UseLoopCountLimit=false)
  ]
  particleSpawnScript [&InitializeParticle(
    Color Maximum=<rgba 0.5837 0.669 0.7948 0.25>
    Color Minimum=<rgba 0.3695 0.4235 0.5031 0.4>
    Color Mode=::NewEnumerator2
    Lifetime=0.45
    Sprite Size Mode=::NewEnumerator3
    Uniform Sprite Size=20
    UsePositionOffset=false
    Write Material Random=true
   )]
  particleUpdateScript [
   &ParticleState()
   &SubUVAnimation(
    "Sprite Renderer "=DataInterface() {propertyValues {SpriteRenderer=&self:Minimal.NiagaraSpriteRendererProperties_0}}
    UseEndFrame=false
    UseStartFrame=false
   )
   &ScaleSpriteSize(Uniform Curve Sprite Scale=DataInterface() {propertyValues {Curve {
       defaultValue=0
       keys [
        {arriveTangent=0.4798, interpMode=RCIM_Cubic, leaveTangent=0.4798, tangentMode=RCTM_User, time=0, value=0.1176}
        {arriveTangent=6.0056, interpMode=RCIM_Cubic, leaveTangent=6.0056, tangentMode=RCTM_User, time=0.7323, value=1.8354}
        {arriveTangent=6.414, interpMode=RCIM_Cubic, leaveTangent=6.414, tangentMode=RCTM_User, time=0.9135, value=3.4091}
        <at 0.9964 3.6009 RCIM_Cubic>
       ]
      }}})
   &ScaleColor(
    Scale Alpha=&FloatFromCurve(FloatCurve=DataInterface() {propertyValues {Curve {
        defaultValue=1
        keys [
         {interpMode=RCIM_Cubic, tangentMode=RCTM_User, time=0, value=0}
         <at 0.1258 0.998 RCIM_Cubic>
         {interpMode=RCIM_Cubic, tangentMode=RCTM_User, time=1, value=0}
        ]
       }}})
    ScaleA=true
    ScaleRGB=true
   )
   /SetVariables(Particles.Position=&AddVectorToPosition(
     Position=$Particles.Position
     Vector=&MakeVector(
      X=0
      Y=&FloatFromCurve(
       FloatCurve=DataInterface() {propertyValues {Curve {keys [
           {interpMode=RCIM_Cubic, tangentMode=RCTM_User, time=0, value=1}
           {arriveTangent=-0.6867, interpMode=RCIM_Cubic, leaveTangent=-0.6867, tangentMode=RCTM_User, time=1, value=0}
          ]}}}
       Scale Curve=0.45
      )
      Z=0
     )
    )) {isSetParameters=true}
  ]
  rendererClasses [&NiagaraSpriteRendererProperties]
  renderers [{rendererClass=&NiagaraSpriteRendererProperties}]
 }
 Minimal001 {
  bEnabled=false
  emitterUpdateScript [
   &EmitterState(Life Cycle Mode=::NewEnumerator1, Loop Behavior=::NewEnumerator1, Loop Duration=2, UseLoopDelay=false)
   &SpawnBurst_Instantaneous(Use Spawn Probability=false, UseLoopCountLimit=false)
   &SpawnBurst_Instantaneous(Spawn Group=1, Use Spawn Probability=false, UseLoopCountLimit=false) {module=SpawnBurst_Instantaneous001}
  ]
  particleSpawnScript [
   /SetVariables(Particles.VisibilityTag=&SpawnGroup()) {isSetParameters=true}
   &InitializeParticle(
    Color=&MergeHelper_ColorSwitch(
     0=<rgba 0.6913 0.7468 0.7615 0.15>
     1=<rgba 0.7064 0.7631 0.7781 1>
     2=?
     3=?
     Index=&SpawnGroup()
     Num=::NewEnumerator0
    )
    Lifetime=0.4
    Mesh Scale=&MergeHelper_VectorSwitch(0=<0.18 0.18 0.5>, 1=<0.12 0.07 0.08>, 2=?, 3=?, Index=&SpawnGroup(), Num=::NewEnumerator0)
    Mesh Scale Mode=::NewEnumerator4
    Position Offset=&MergeHelper_VectorSwitch(0=<0 -1 0>, 1=<0 -1 0>, 2=?, 3=?, Index=&SpawnGroup(), Num=::NewEnumerator0)
    UsePositionOffset=true
    Write Material Random=true
   )
   &InitialMeshOrientation(
    Mesh Axis To Orient=&MergeHelper_VectorSwitch(0=<0 1 0>, 1=<0 1 0>, 2=?, 3=?, Index=&SpawnGroup(), Num=::NewEnumerator0)
    Mesh Orientation Mode=::NewEnumerator2
    Orientation Vector=&MergeHelper_VectorSwitch(0=<0 0 1>, 1=<1 0 0>, 2=?, 3=?, Index=&SpawnGroup(), Num=::NewEnumerator0)
    Use Rotation Vector=false
   )
  ]
  particleUpdateScript [
   &ParticleState()
   &ScaleMeshSize(Scale Factor=&VectorFromFloat(Value=&FloatFromCurve(
      FloatCurve=DataInterface() {propertyValues {Curve {
         defaultValue=1
         keys [
          {arriveTangent=0.3747, interpMode=RCIM_Cubic, leaveTangent=0.3747, tangentMode=RCTM_User, time=0, value=0}
          {arriveTangent=1.6729, interpMode=RCIM_Cubic, leaveTangent=1.6729, tangentMode=RCTM_User, time=0.642, value=1.5746}
          <at 0.9994 1.9694 RCIM_Cubic>
         ]
        }}}
      Scale Curve=0.7
     )))
   &DynamicMaterialParameters(
    Index 0 Param 1=&FloatFromCurve(FloatCurve=DataInterface() {propertyValues {Curve {keys [<at 0 1>,<at 1 0>]}}})
    Index 0 Param 2=&FloatFromCurve(FloatCurve=DataInterface() {propertyValues {Curve {defaultValue=1, keys [{tangentMode=RCTM_User, time=0, value=0}, {tangentMode=RCTM_User, time=1, value=1}]}}})
    Index 0 Param 4=&FloatFromCurve(FloatCurve=DataInterface() {propertyValues {Curve {keys [<at 0 0>,<at 0.03 1>,<at 0.77 1>,<at 1 0>]}}})
    Module.Index0Float0Write=true
    Module.Index0Float1Write=true
    Module.Index0Float2Write=true
    Module.Index0Float3Write=true
   )
   &ScaleColor(
    Scale Alpha=&FloatFromCurve(FloatCurve=DataInterface() {propertyValues {Curve {
        defaultValue=1
        keys [
         {interpMode=RCIM_Cubic, tangentMode=RCTM_User, time=0, value=0}
         <at 0.1258 0.998 RCIM_Cubic>
         {interpMode=RCIM_Cubic, tangentMode=RCTM_User, time=1, value=0}
        ]
       }}})
    ScaleA=true
    ScaleRGB=true
   )
   /SetVariables(Particles.Position=&AddVectorToPosition(
     Position=$Particles.Position
     Vector=&MakeVector(
      X=0
      Y=&FloatFromCurve(
       FloatCurve=DataInterface() {propertyValues {Curve {keys [
           {interpMode=RCIM_Cubic, tangentMode=RCTM_User, time=0, value=1}
           {arriveTangent=-0.6867, interpMode=RCIM_Cubic, leaveTangent=-0.6867, tangentMode=RCTM_User, time=1, value=0}
          ]}}}
       Scale Curve=0.45
      )
      Z=0
     )
    )) {isSetParameters=true}
  ]
  rendererClasses [&NiagaraMeshRendererProperties]
  renderers [
   {rendererClass=&NiagaraMeshRendererProperties, rendererIndex=0}
   {rendererClass=&NiagaraMeshRendererProperties, rendererIndex=1}
  ]
 }
}
summary {
 systemName=NS_Emote_Sitcom_Glow
 emitters [
  {bEnabled=false, emitterName=Minimal, rendererClasses [&NiagaraSpriteRendererProperties]}
  {bEnabled=false, emitterName=Minimal001, rendererClasses [&NiagaraMeshRendererProperties]}
  {bEnabled=false, emitterName=Little_Lines, rendererClasses [&NiagaraSpriteRendererProperties]}
  {
   bEnabled=false
   emitterName=Glowy
   rendererClasses [&NiagaraSpriteRendererProperties, &NiagaraMeshRendererProperties]
  }
  {bEnabled=false, emitterName=Barrier_PulseBuildUp, rendererClasses [&NiagaraMeshRendererProperties]}
  {bEnabled=false, emitterName=Barrier_Release, rendererClasses [&NiagaraMeshRendererProperties]}
  {
   emitterName=Condensed_ChestGlow
   rendererClasses [&NiagaraSpriteRendererProperties, &NiagaraMeshRendererProperties]
  }
 ]
 userVariables [
  User.bIsFrontEnd {type=NiagaraBool}
  User.bIsFrontEndPreview {type=NiagaraBool}
  User.Intensity {type=NiagaraFloat, defaultValue=1, description=""}
 ]
}
