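# Niagara system that draws a ribbon hand trail for the Sitcom emote.
# Its single emitter spawns anim-trail particles from skeletal mesh bones, with a time-of-day tinted, translucent blue color.
# A user float named for an animation curve scales particle lifetime, despite its SpawnRate name.
# Ribbon width tapers by curve, while alpha fades on emitter age rather than particle age.
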NiagaraSystem /BRCosmetics/Animation/Game/MainPlayer/Emotes/Sitcom/FX/NS_Emote_Sitcom_HandTrail
dependencies {
 usedDataInterfaces [&NiagaraDataInterfaceCurve, &NiagaraDataInterfaceSkeletalMesh]
 usedDynamicInputs []
 usedModules [
  &self:SystemSpawnScript.SystemScriptSource.SystemScriptGraph.NiagaraNodeAssignment_0.SetVariables_#1
  &self:SystemSpawnScript.SystemScriptSource.SystemScriptGraph.NiagaraNodeAssignment_1.SetVariables_#2
  &self:SystemSpawnScript.SystemScriptSource.SystemScriptGraph.NiagaraNodeAssignment_2.SetVariables_#3
  &SystemState
 ]
}
emitters {Hand_AnimTrail {
  emitterUpdateScript [
   &EmitterState(Life Cycle Mode=::NewEnumerator1, Loop Behavior=::NewEnumerator1, Loop Duration=2, UseLoopDelay=false)
   &FN_SpawnAnimTrail(Spawn Mode:ENiagaraSkeletalSpawning=::NewEnumerator0)
  ]
  particleSpawnScript [
   &FN_InitializeAnimTrailBoneIndecies(Second Bone or Socket Index=2)
   &FN_InitializeAnimTrail(
    AnimTrailWidth=0.17
    Color=&FNTimeOfDayColor(
     Night Color=<rgba 0.2908 0.6227 0.7948 0.3>
     Morning Color=<rgba 0.2908 0.6227 0.7948 0.4>
     Day Color=<rgba 0.2908 0.6227 0.7948 0.5>
     Evening Color=<rgba 0.2908 0.6227 0.7948 0.4>
    )
    Lifetime=&Multiply_Float(
     A=&FloatFromCurve(
      FloatCurve=DataInterface() {propertyValues {Curve {
         defaultValue=1
         keys [{interpMode=RCIM_Cubic, tangentMode=RCTM_User, time=0, value=1}, {interpMode=RCIM_Cubic, tangentMode=RCTM_User, time=0.3, value=-0.0601}]
        }}}
      Scale Curve=1.05
     )
     B=$User.Anim_Curve_SpawnRate_Ribbon
    )
    Position Offset=<-2 0 2>
    SkeletalMesh=$System.Skeletal Mesh
   )
   &ApplyInitialForces()
   &SkeletalMeshLocation(
    Bone / Socket Index=$Particles.PartitionParticles.Partition
    Bone Sampling Mode=::NewEnumerator4
    MaskSpawnGroup=false
    Skeletal Mesh=$System.Skeletal Mesh
   )
  ]
  particleUpdateScript [
   &ParticleState()
   &FN_UpdateAnimTrail(SkeletalMesh=$System.Skeletal Mesh)
   &ScaleRibbonWidth(Ribbon Width Scale=&FloatFromCurve(FloatCurve=DataInterface() {propertyValues {Curve {
        defaultValue=1
        keys [
         {arriveTangent=0.0733, interpMode=RCIM_Cubic, leaveTangent=0.0733, tangentMode=RCTM_User, time=0, value=1}
         {arriveTangent=-2.5215, interpMode=RCIM_Cubic, leaveTangent=-2.5215, tangentMode=RCTM_User, time=0.4, value=0.3}
         <at 0.599 0 RCIM_Cubic>
        ]
       }}}))
   &InheritSourceMovement(ApplyForce=false, ApplyPosition=true, ApplyPositionFalloff=false)
   &SolveForcesAndVelocity(Clamp Velocity=false, Limit Acceleration=false)
   &ScaleColor(
    Scale Alpha=&FloatFromCurve(
     FloatCurve=DataInterface() {propertyValues {Curve {
        defaultValue=1
        keys [
         {interpMode=RCIM_Cubic, tangentMode=RCTM_User, time=0, value=1}
         {arriveTangent=-4.4643, interpMode=RCIM_Cubic, leaveTangent=-4.4643, tangentMode=RCTM_User, time=0.96, value=0}
        ]
       }}}
     CurveIndex=$Emitter.Age
    )
    ScaleA=true
    ScaleRGB=true
   )
  ]
  rendererClasses [&NiagaraRibbonRendererProperties]
  renderers [{rendererClass=&NiagaraRibbonRendererProperties}]
 }}
summary {
 systemName=NS_Emote_Sitcom_HandTrail
 emitters [{emitterName=Hand_AnimTrail, rendererClasses [&NiagaraRibbonRendererProperties]}]
 userVariables [
  User.Anim_Curve_SpawnRate_Ribbon {type=NiagaraFloat, defaultValue=1, description=""}
  User.bIsFrontEnd {type=NiagaraBool}
  User.bIsFrontEndPreview {type=NiagaraBool}
  User.RibbonIDOffset {type=NiagaraInt32}
 ]
}
