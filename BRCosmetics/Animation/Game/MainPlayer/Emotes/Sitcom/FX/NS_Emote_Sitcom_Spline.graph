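# Niagara effect for the Sitcom emote that draws a stretched mesh link between two sampled sockets, apparently chest and hand given the emitter name.
# Spline_ChestHand, the only enabled emitter, bursts two spawn groups tagged for its two mesh renderers and reads both sockets through a system Socket Reader.
# Each update orients the meshes along the socket-to-socket vector, scales Z by that distance, and shifts HDR color from blue through orange to red.
# Three other mesh and sprite emitters are disabled.
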NiagaraSystem /BRCosmetics/Animation/Game/MainPlayer/Emotes/Sitcom/FX/NS_Emote_Sitcom_Spline
dependencies {
 usedDataInterfaces [
  &NiagaraDataInterfaceCurve
  &NiagaraDataInterfaceSpriteRendererInfo
  &NiagaraDataInterfaceColorCurve
  &NiagaraDataInterfaceSocketReader
 ]
 usedDynamicInputs []
 usedModules [
  &SystemState
  &self:SystemSpawnScript.SystemScriptSource.SystemScriptGraph.NiagaraNodeAssignment_0.SetVariables_#1
 ]
}
emitters {
 Little_Lines {
  bEnabled=false
  emitterUpdateScript [
   &EmitterState()
   &SpawnBurst_Instantaneous(
    Spawn Count=&RandomRangeInt(Minimum=4, Maximum=7)
    Use Spawn Probability=false
    UseLoopCountLimit=false
   )
   &SpawnRate(SpawnRate=400, Use Spawn Probability=false) {enabled=false}
  ]
  particleSpawnScript [
   &InitializeParticle(
    Color=&RandomRangeLinearColor(Minimum=<rgba 0.3288 0.3419 0.4226 0.3>, Maximum=<rgba 0.4031 0.4031 0.4991 0.42>)
    Lifetime Max=0.34
    Lifetime Min=0.2
    Lifetime Mode=::NewEnumerator1
    Sprite Size Max=<0.9 1.3>
    Sprite Size Min=<0.6 0.9>
    Sprite Size Mode=::NewEnumerator2
    UsePositionOffset=true
    Write Material Random=true
   ) {module=InitializeParticle001}
   &AddVelocity(Velocity=&RandomRangeVector(Minimum=<0 15 0>, Maximum=<0 30 0>, Evaluation Type=::NewEnumerator1))
   &ShapeLocation(
    Handle Radius=0.01
    Large Radius=0.5
    MaskSpawnGroup=false
    Override Seed=false
    Rotation Axis=<0 1 0>
    Rotation Mode=::NewEnumerator1
    Shape Primitive=::NewEnumerator3
    UseOwnerScale=true
   )
   &AddVelocity(
    Module.Use Falloff Distance=false
    Override Seed=false
    Velocity Mode=::NewEnumerator1
    Velocity Speed=&RandomRangeFloat(Minimum=10, Maximum=15)
   ) {module=AddVelocity001}
  ]
  particleUpdateScript [
   &ParticleState()
   &ScaleSpriteSize(Scale Sprite Size Mode=::NewEnumerator0)
   &SolveForcesAndVelocity(Clamp Velocity=false, Limit Acceleration=false)
   &ScaleSpriteSizeBySpeed(
    Max Scale Factor=<0.4 10>
    Min Scale Factor=<0.25 5>
    Source Velocity:Vector3f=&Multiply_Vector(A=<0.05 0.05 1>, B=$Particles.Velocity)
    Velocity Threshold=500
   )
  ]
  rendererClasses [&NiagaraSpriteRendererProperties]
  renderers [{rendererClass=&NiagaraSpriteRendererProperties}]
 }
 Minimal002 {
  bEnabled=false
  emitterUpdateScript [
   &EmitterState(Life Cycle Mode=::NewEnumerator1, Loop Behavior=::NewEnumerator1, Loop Duration=2, UseLoopDelay=false)
   &SpawnBurst_Instantaneous(Use Spawn Probability=false, UseLoopCountLimit=false)
   &SpawnBurst_Instantaneous(Spawn Group=1, Use Spawn Probability=false, UseLoopCountLimit=false) {module=SpawnBurst_Instantaneous001}
  ]
  particleSpawnScript [
   /SetVariables(Particles.VisibilityTag=&SpawnGroup()) {isSetParameters=true}
   &InitializeParticle(
    Color=&MergeHelper_ColorSwitch(
     0=<rgba 0.7896 0.853 0.8698 0.33>
     1=<rgba 0.7064 0.7631 0.7781 2>
     2=?
     3=?
     Index=&SpawnGroup()
     Num=::NewEnumerator0
    )
    Lifetime=0.4
    Mesh Scale=&MergeHelper_VectorSwitch(0=<0.1 0.1 0.14>, 1=<0.07 0.04 0.045>, 2=?, 3=?, Index=&SpawnGroup(), Num=::NewEnumerator0)
    Mesh Scale Mode=::NewEnumerator4
    Position Offset=&MergeHelper_VectorSwitch(0=<0 -0.5 0>, 1=<0 -2 0>, 2=?, 3=?, Index=&SpawnGroup(), Num=::NewEnumerator0)
    UsePositionOffset=true
    Write Material Random=true
   )
   &InitialMeshOrientation(
    Mesh Axis To Orient=&MergeHelper_VectorSwitch(0=<0 1 0>, 1=<0 1 0>, 2=?, 3=?, Index=&SpawnGroup(), Num=::NewEnumerator0)
    Mesh Orientation Mode=::NewEnumerator2
    Orientation Vector=&MergeHelper_VectorSwitch(0=<0 0 1>, 1=<1 0 0>, 2=?, 3=?, Index=&SpawnGroup(), Num=::NewEnumerator0)
    Use Rotation Vector=false
   )
  ]
  particleUpdateScript [
   &ParticleState()
   &ScaleMeshSize(Scale Factor=&VectorFromFloat(Value=&FloatFromCurve(
      FloatCurve=DataInterface() {propertyValues {Curve {
         defaultValue=1
         keys [
          {arriveTangent=1.0928, interpMode=RCIM_Cubic, leaveTangent=1.0928, tangentMode=RCTM_User, time=0, value=0}
          {arriveTangent=0.5906, interpMode=RCIM_Cubic, leaveTangent=6.9979, tangentMode=RCTM_Break, time=0.6746, value=1.0556}
          {arriveTangent=4.3958, interpMode=RCIM_Cubic, leaveTangent=4.3958, time=0.8038, value=2.1176}
          <at 1.0048 2.5069 RCIM_Cubic>
         ]
        }}}
      Scale Curve=0.7
     )))
   &DynamicMaterialParameters(
    Index 0 Param 1=&FloatFromCurve(FloatCurve=DataInterface() {propertyValues {Curve {keys [<at 0 1>,<at 1 0>]}}})
    Index 0 Param 2=&FloatFromCurve(FloatCurve=DataInterface() {propertyValues {Curve {defaultValue=1, keys [{tangentMode=RCTM_User, time=0, value=0}, {tangentMode=RCTM_User, time=1, value=1}]}}})
    Index 0 Param 4=&FloatFromCurve(FloatCurve=DataInterface() {propertyValues {Curve {keys [<at 0 0>,<at 0.03 1>,<at 0.77 1>,<at 1 0>]}}})
    Module.Index0Float0Write=true
    Module.Index0Float1Write=true
    Module.Index0Float2Write=true
    Module.Index0Float3Write=true
   )
   &ScaleColor(
    Scale Alpha=&FloatFromCurve(FloatCurve=DataInterface() {propertyValues {Curve {
        defaultValue=1
        keys [
         {interpMode=RCIM_Cubic, tangentMode=RCTM_User, time=0, value=0}
         <at 0.1258 0.998 RCIM_Cubic>
         {interpMode=RCIM_Cubic, tangentMode=RCTM_User, time=1, value=0}
        ]
       }}})
    ScaleA=true
    ScaleRGB=true
   )
   /SetVariables(Particles.Position=&AddVectorToPosition(
     Position=$Particles.Position
     Vector=&MakeVector(
      X=0
      Y=&FloatFromCurve(
       FloatCurve=DataInterface() {propertyValues {Curve {keys [
           {interpMode=RCIM_Cubic, tangentMode=RCTM_User, time=0, value=1}
           {arriveTangent=-0.6867, interpMode=RCIM_Cubic, leaveTangent=-0.6867, tangentMode=RCTM_User, time=1, value=0}
          ]}}}
       Scale Curve=0.45
      )
      Z=0
     )
    )) {isSetParameters=true}
  ]
  rendererClasses [&NiagaraMeshRendererProperties]
  renderers [
   {rendererClass=&NiagaraMeshRendererProperties, rendererIndex=0}
   {rendererClass=&NiagaraMeshRendererProperties, rendererIndex=1}
  ]
 }
 Minimal003 {
  bEnabled=false
  emitterUpdateScript [
   &EmitterState(Life Cycle Mode=::NewEnumerator1, Loop Behavior=::NewEnumerator1, Loop Duration=2, UseLoopDelay=false)
   &SpawnBurst_Instantaneous(Use Spawn Probability=false, UseLoopCountLimit=false)
  ]
  particleSpawnScript [&InitializeParticle(
    Color=<rgba 0.524 0.6006 0.7135 0.3>
    Lifetime=0.45
    Sprite Size Mode=::NewEnumerator3
    Uniform Sprite Size=20
    UsePositionOffset=false
    Write Material Random=true
   )]
  particleUpdateScript [
   &ParticleState()
   &SubUVAnimation(
    "Sprite Renderer "=DataInterface() {propertyValues {SpriteRenderer=&self:Minimal003_0.NiagaraSpriteRendererProperties_0}}
    UseEndFrame=false
    UseStartFrame=false
   )
   &ScaleSpriteSize(
    Uniform Curve Scale=0.5
    Uniform Curve Sprite Scale=DataInterface() {propertyValues {Curve {
       defaultValue=0
       keys [
        {arriveTangent=0.4798, interpMode=RCIM_Cubic, leaveTangent=0.4798, tangentMode=RCTM_User, time=0, value=0.1176}
        {arriveTangent=3.63, interpMode=RCIM_Cubic, leaveTangent=3.63, tangentMode=RCTM_User, time=0.7563, value=2.3461}
        <at 0.9994 4.2759 RCIM_Cubic>
       ]
      }}}
   )
   &ScaleColor(
    Scale Alpha=&FloatFromCurve(FloatCurve=DataInterface() {propertyValues {Curve {
        defaultValue=1
        keys [
         {interpMode=RCIM_Cubic, tangentMode=RCTM_User, time=0, value=0}
         <at 0.1258 0.998 RCIM_Cubic>
         {interpMode=RCIM_Cubic, tangentMode=RCTM_User, time=1, value=0}
        ]
       }}})
    ScaleA=true
    ScaleRGB=true
   )
   /SetVariables(Particles.Position=&AddVectorToPosition(
     Position=$Particles.Position
     Vector=&MakeVector(
      X=0
      Y=&FloatFromCurve(
       FloatCurve=DataInterface() {propertyValues {Curve {keys [
           {interpMode=RCIM_Cubic, tangentMode=RCTM_User, time=0, value=1}
           {arriveTangent=-0.6867, interpMode=RCIM_Cubic, leaveTangent=-0.6867, tangentMode=RCTM_User, time=1, value=0}
          ]}}}
       Scale Curve=0.45
      )
      Z=0
     )
    )) {isSetParameters=true}
  ]
  rendererClasses [&NiagaraSpriteRendererProperties]
  renderers [{rendererClass=&NiagaraSpriteRendererProperties}]
 }
 Spline_ChestHand {
  emitterUpdateScript [
   &EmitterState(
    Inactive Response=::NewEnumerator1
    Life Cycle Mode=::NewEnumerator1
    Loop Behavior=::NewEnumerator1
    Loop Duration Mode=::NewEnumerator1
    UseLoopDelay=false
   )
   &SpawnBurst_Instantaneous(Use Spawn Probability=false, UseLoopCountLimit=false)
   &SpawnBurst_Instantaneous(Spawn Group=1, Use Spawn Probability=false, UseLoopCountLimit=false) {module=SpawnBurst_Instantaneous001}
  ]
  particleSpawnScript [
   /SetVariables(Particles.VisibilityTag=&SpawnGroup()) {isSetParameters=true}
   &InitializeParticle(
    Color Mode=::NewEnumerator0
    Lifetime=1.84
    Mesh Scale Mode=::NewEnumerator3
    Position Offset=<0 -0.5 0>
    UsePositionOffset=false
    Write Material Random=true
   )
   &InitialMeshOrientation(Mesh Orientation Mode=::NewEnumerator2, Orientation Vector=<0 0 1>, Use Rotation Vector=false)
   &CameraOffset(Camera Offset Amount=2) {enabled=false}
  ]
  particleUpdateScript [
   &ParticleState(Kill Particles When Lifetime Has Elapsed=false, Let Infinitely Lived Particles Die When Emitter Deactivates=true)
   &SocketLocation(
    Orientation Sampling=::NewEnumerator1
    Position Sampling=::NewEnumerator1
    SamplePosOffset=false
    Sampled Position Coordinate Space:ENiagaraCoordinateSpace=::Local
    Socket Reader=$System.Socket Reader
    Socket Sampling Mode=::NewEnumerator3
    VelocitySpeedLimit=false
   )
   &SocketLocation(
    Orientation Sampling=::NewEnumerator1
    SamplePosOffset=true
    Sampled Position Coordinate Space:ENiagaraCoordinateSpace=::Local
    Sampled Position Offset=<-7 3 0>
    Socket Index=1
    Socket Reader=$System.Socket Reader
    Socket Sampling Mode=::NewEnumerator3
    VelocitySpeedLimit=false
   ) {module=SocketLocation001}
   &ScaleColor(
    Linear Color Curve=DataInterface() {propertyValues {
      AlphaCurve {keys [<at 0 1>]}
      BlueCurve {keys [<at 0.1267 10>,<at 0.4098 0>,<at 0.5414 0>,<at 0.8209 0>]}
      GreenCurve {keys [<at 0.1267 4.6495>,<at 0.4098 3.1701>,<at 0.5414 3.1701>,<at 0.8209 0.1726>]}
      RedCurve {keys [<at 0.1267 0.0375>,<at 0.4098 10>,<at 0.5414 10>,<at 0.8209 15>]}
     }}
    Scale Mode=::NewEnumerator2
   )
   /SetVariables(Particles.New Mesh Orientation:Quat4f=&OrientMeshToVector(
     Look At Direction=&Subtract_Position(A=$Output.SocketLocation001.SampledPosition, B=$Output.SocketLocation.SampledPosition)
     Reference Axis Vector=<1 0 0>
    )) {isSetParameters=true}
   &ScaleMeshSize(Scale Factor=&MakeVector(
     X=1
     Y=1
     Z=&Divide_Float(
      A=&VectorLength(Vector=&Subtract_Vector(
        A=Linked() {linkedVariable {name=Output.SocketLocation001.SampledPosition, type=NiagaraPosition}}
        B=Linked() {linkedVariable {name=Output.SocketLocation.SampledPosition, type=NiagaraPosition}}
       ))
      B=20
     )
    ))
  ]
  rendererClasses [&NiagaraMeshRendererProperties]
  renderers [
   {rendererClass=&NiagaraMeshRendererProperties, rendererIndex=0}
   {rendererClass=&NiagaraMeshRendererProperties, rendererIndex=1}
  ]
 }
}
summary {
 systemName=NS_Emote_Sitcom_Spline
 emitters [
  {bEnabled=false, emitterName=Minimal002, rendererClasses [&NiagaraMeshRendererProperties]}
  {bEnabled=false, emitterName=Minimal003, rendererClasses [&NiagaraSpriteRendererProperties]}
  {bEnabled=false, emitterName=Little_Lines, rendererClasses [&NiagaraSpriteRendererProperties]}
  {emitterName=Spline_ChestHand, rendererClasses [&NiagaraMeshRendererProperties]}
 ]
 userVariables [
  User.ForHandSample {
   type=NiagaraDataInterfaceSkeletalMesh
   defaultValue {
    struct=NiagaraExt_VariableValue_DataInterface
    value {dataInterface=&self:NiagaraDataInterfaceSkeletalMesh_0}
   }
  }
  User.Intensity {type=NiagaraFloat, defaultValue=1, description=""}
 ]
}
